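# Parent material for the VibrantShell female character body, adding fake depth shading to its wristband on top of the standard 2023 character PreFX and PostFX functions.
# Pre-skinned position is projected onto a parameterized ring plane and crossed with the axis to derive a tangent for an axis-aligned Fresnel.
# That term times a regular Fresnel scales base color between deep and thin brightness, blended in through a Tech Art Mask channel.
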Material /BRCosmetics/Characters/Player/Female/Medium/Bodies/F_MED_VibrantShell/Materials/M_VibrantShell_Body_Parent
graph {
 properties {
  bAutomaticallySetUsageInEditor=false
  bUseMaterialAttributes=true
  bUsedWithMeshDeformer=true
  bUsedWithMorphTargets=true
  bUsedWithSkeletalMesh=true
  pixelDepthOffsetMode=PDOM_Legacy
  shadingModel=MSM_FromMaterialExpression
  usedShadingModels=Default Lit | Subsurface | Clear Coat | Subsurface Profile | Cloth
 }
 nodes [
  n1 = MaterialFunctionCall() {materialFunction=&MF_CharacterParent_2023_PreFX}
  n2 = GetMaterialAttributes(None=@n1.Result) {attributeGetTypes [#1]}
  n3 = Normalize(VectorInput=VectorParameter() {group=VibrantShell, parameterName=Wristband_RingAxis, defaultValue=<rgba 1 0.8 -1 1>}.RGB)
  n4 = LocalPosition() {localOrigin=InstancePreSkinning}
  n5 = MaterialFunctionCall!(
   Plane Normal=@n3
   Vector=Subtract(
    A=VertexInterpolator(VS=@n4.XYZ).PS
    B=VectorParameter() {group=VibrantShell, parameterName=Wristband_RingPos, defaultValue=<rgba 48 3.5 105 1>}.RGB
   )
  ) {materialFunction=/Engine/Functions/Engine_MaterialFunctions02/Math/ProjectVectorOntoPlane}
  n6 = MaterialFunctionCall!(
   Axis=CrossProduct(A=Normalize(VectorInput=@n5.Result), B=@n3)
   Exponent=ScalarParameter() {defaultValue=3, group=VibrantShell, parameterName=WristbandFresnelExp_AxisAligned}
   Normal=PixelNormalWS()
  ) {materialFunction=/Game/Effects/Fort_Effects/Materials/_Functions/MF_AxisAlignedFrensnel}
  n7 = Fresnel(ExponentIn=ScalarParameter() {defaultValue=0.8, group=VibrantShell, parameterName=WristbandFresnelExp}) {baseReflectFraction=0}
  n8 = Multiply(
   A=@n2.Base Color
   B=LinearInterpolate(
    A=ScalarParameter() {defaultValue=0.25, group=VibrantShell, parameterName=WristbandBrightness_Deep}
    B=ScalarParameter() {defaultValue=1, group=VibrantShell, parameterName=WristbandBrightness_Thin}
    Alpha=Multiply(A=@n6.Fresnel, B=@n7)
   )
  )
  n9 = ChannelMaskParameter(None=@n1.Tech Art Mask) {group=VibrantShell, maskChannel=Blue, parameterName=WristBandMaskChannel, defaultValue=<rgba 0 0 1 0>}
  MaterialFunctionCall(
   Material=BlendMaterialAttributes(
    A=@n1.Result
    B=SetMaterialAttributes!(MaterialAttributes=@n2.MaterialAttributes, Base Color=@n8) {attributeSetTypes [#1]}
    Alpha=@n9.R
   ) {vertexAttributeBlendType=UseA}
   M=@n1.M
   Pre Skinned Local Position=@n1.M
   Distance Blend=@n1.M
   Fresnel=@n1.M
  ) {materialFunction=&MF_CharacterParent_2023_PostFX}
 ]
}
