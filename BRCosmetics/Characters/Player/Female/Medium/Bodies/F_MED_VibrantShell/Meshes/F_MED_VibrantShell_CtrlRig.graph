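# Control Rig for the VibrantShell female medium body cosmetic, adding pose-driven corrections to deform and accessory bones.
# A shared DeformRig function runs first, then sequenced branches adjust left bracelets, right upper arm and both shoe butterfly wings.
# Spherical pose readers weight additive transform offsets, with is_FE scaling one upper-arm correction; left bracelets also get X-only rotation constraints to the hand.
# Right-bracelet readers and one extra upper-arm reader are disconnected from execution.
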ControlRigBlueprint /BRCosmetics/Characters/Player/Female/Medium/Bodies/F_MED_VibrantShell/Meshes/F_MED_VibrantShell_CtrlRig
graphs [
 RigVMFunctionLibrary {links [], nodes []}
 RigVMFunction_Sequence_9_ContainedGraph {nodes [
   n1 = RigVMFunctionEntryNode()
   n2 = &RigVMFunction_Sequence(ExecuteContext=@n1.ExecuteContext)
   n3 = &RigVMFunction_Sequence(ExecuteContext=@n2.B)
   RigVMFunctionReturnNode(B=@n3.A, C=@n3.B, A=@n2.A)
  ]}
 RigVMModel {nodes [
   RigUnit_BeginExecution = &RigUnit_BeginExecution()
   FN_DeformRig_FNC = RigVMFunctionReferenceNode(
    DeformRig_ Curve="(deformRig_l_upperarm_up_alpha_72_#1=False,deformRig_r_upperarm_up_alpha_73_#2=False,deformRig_l_upperarm_fr_alpha_74_#3=False,deformRig_r_upperarm_fr_alpha_75_CB51B8 ...(truncated)"
    DeformRig_Bool="(upperarm_1_#4=True,elbow_3_#5=True,wrist_7_#6=True,groin_8_#7=True,glute_17_#8=True,knee_9_40C408324CA261 ...(truncated)"
    DeformRig_Debug=(upperarm_debug_18_#4=False,wrist_debug_20_#6=False,lat_debug_26_#9=False,groinGlute_debug_27_#7=False)
    DeformRig_Value="(upperarm_up_18_#10=(Rotation=(X=0,Y=0,Z=0,W=1),Translation=(X=0,Y=0,Z=8),Scale3D=(X=0.7,Y=1,Z=1)),upperarm_fwd_fr_19_#11=(Rotation=(X=0,Y=0,Z=0,W=1),Translation=(X=0,Y=10,Z=0),Scale3D=(X=0.5,Y=1,Z= ...(truncated)"
    ExecuteContext=@RigUnit_BeginExecution.ExecutePin
   )
   n1 = &RigVMFunction_Sequence(ExecuteContext=@FN_DeformRig_FNC.ExecuteContext)
   n2 = &RigVMFunction_Sequence(ExecuteContext=@n1.A)
   &RigUnit_RotationConstraintLocalSpaceOffset(
    AdvancedSettings=(InterpolationType=Average,RotationOrderForFilter=XZY)
    Child=(Type=Bone,Name="dyn_bracelet_1_l")
    Filter=(bX=true,bY=false,bZ=false)
    Parents=((Item=(Type=Bone,Name="hand_l"),Weight=1))
    Weight=1
    ExecutePin=@n2.A
   )
   > n3 = &RigUnit_RotationConstraintLocalSpaceOffset(
    AdvancedSettings=(InterpolationType=Average,RotationOrderForFilter=XZY)
    Child=(Type=Bone,Name="dyn_bracelet_2_l")
    Filter=(bX=true,bY=false,bZ=false)
    Parents=((Item=(Type=Bone,Name="hand_l"),Weight=1))
    Weight=1
   )
   n4 = &RigVMFunction_Sequence(ExecuteContext=@n3.ExecutePin)
   n5 = &RigUnit_SphericalPoseReader(
    ActiveRegionScaleFactors=(PositiveWidth=0.5,NegativeWidth=2,PositiveHeight=2,NegativeHeight=0.5)
    ActiveRegionSize=0.1
    Debug=(bDrawDebug=true,bDraw2D=False,bDrawLocalAxes=False,DebugScale=25,DebugSegments=20,DebugThickness=0.25)
    DriverAxis=(X=1,Y=0,Z=0)
    DriverItem=(Type=Bone,Name="hand_l")
    FalloffRegionScaleFactors=(PositiveWidth=1,NegativeWidth=1,PositiveHeight=1,NegativeHeight=1)
    FalloffSize=0.4
    FlipHeightScaling=False
    FlipWidthScaling=False
    OptionalParentItem=(Type=Bone,Name="None")
    RotationOffset=(X=0,Y=0,Z=0)
    ExecutePin=@n4.A
   )
   > &RigUnit_ModifyTransforms(
    ItemToModify=((Item=(Type=Bone,Name="dyn_bracelet_1_l"),Transform=(Rotation=(X=0,Y=0,Z=0,W=1),Translation=(X=-1.5,Y=0,Z=0),Scale3D=(X=1,Y=1,Z=1))))
    ItemToModify.0.Transform.Rotation=&RigVMFunction_MathQuaternionFromEuler(Euler=(X=0,Y=-35,Z=30), RotationOrder=ZXY).Result
    Weight=@n5.OutputParam
   )
   > &RigUnit_ModifyTransforms(
    ItemToModify=((Item=(Type=Bone,Name="dyn_bracelet_2_l"),Transform=(Rotation=(X=0,Y=0,Z=0,W=1),Translation=(X=-1,Y=0,Z=0),Scale3D=(X=1,Y=1,Z=1))))
    ItemToModify.0.Transform.Rotation=&RigVMFunction_MathQuaternionFromEuler(Euler=(X=0,Y=-10,Z=0), RotationOrder=ZXY).Result
    Weight=@n5.OutputParam
   )
   n6 = &RigUnit_SphericalPoseReader(
    ActiveRegionScaleFactors=(PositiveWidth=1,NegativeWidth=1,PositiveHeight=1,NegativeHeight=1)
    ActiveRegionSize=0.1
    Debug=(bDrawDebug=true,bDraw2D=False,bDrawLocalAxes=False,DebugScale=25,DebugSegments=20,DebugThickness=0.25)
    DriverAxis=(X=1,Y=0,Z=0)
    DriverItem=(Type=Bone,Name="hand_l")
    FalloffRegionScaleFactors=(PositiveWidth=1,NegativeWidth=1,PositiveHeight=1,NegativeHeight=1)
    FalloffSize=0.35
    FlipHeightScaling=False
    FlipWidthScaling=False
    OptionalParentItem=(Type=Bone,Name="None")
    RotationOffset=(X=-20,Y=-150,Z=0)
    ExecutePin=@n4.B
   )
   > &RigUnit_ModifyTransforms(
    ItemToModify=((Item=(Type=Bone,Name="dyn_bracelet_1_l"),Transform=(Rotation=(X=0,Y=0,Z=0,W=1),Translation=(X=-1.35,Y=-0.25,Z=0.25),Scale3D=(X=1,Y=1,Z=1))))
    ItemToModify.0.Transform.Rotation=&RigVMFunction_MathQuaternionFromEuler(Euler=(X=0,Y=-10,Z=17.5), RotationOrder=ZXY).Result
    Weight=@n6.OutputParam
   )
   > &RigUnit_ModifyTransforms(
    ItemToModify=((Item=(Type=Bone,Name="dyn_bracelet_2_l"),Transform=(Rotation=(X=0,Y=0,Z=0,W=1),Translation=(X=-0.5,Y=-0.25,Z=-0.25),Scale3D=(X=1,Y=1,Z=1))))
    ItemToModify.0.Transform.Rotation=&RigVMFunction_MathQuaternionFromEuler(Euler=(X=0,Y=5,Z=5), RotationOrder=ZXY).Result
    Weight=@n6.OutputParam
   )
   n7 = &RigUnit_SphericalPoseReader(
    ActiveRegionScaleFactors=(PositiveWidth=0.5,NegativeWidth=0.5,PositiveHeight=0.5,NegativeHeight=0.5)
    ActiveRegionSize=0.1
    Debug=(bDrawDebug=false,bDraw2D=False,bDrawLocalAxes=False,DebugScale=25,DebugSegments=20,DebugThickness=0.25)
    DriverAxis=(X=0,Y=-1,Z=0)
    DriverItem=(Type=Bone,Name="upperarm_twist_01_r")
    FalloffRegionScaleFactors=(PositiveWidth=1,NegativeWidth=1,PositiveHeight=1,NegativeHeight=1)
    FalloffSize=0.4
    FlipHeightScaling=False
    FlipWidthScaling=False
    OptionalParentItem=(Type=Bone,Name="None")
    RotationOffset=(X=0,Y=0,Z=0)
    ExecutePin=@n2.B
   )
   > &RigUnit_ModifyTransforms(
    ItemToModify=((Item=(Type=Bone,Name="dyn_upperarm_twist_01_corr_r"),Transform=(Rotation=(X=0,Y=0,Z=0,W=1),Translation=(X=0,Y=0,Z=0),Scale3D=(X=1,Y=1,Z=1))))
    ItemToModify.0.Transform.Rotation=&RigVMFunction_MathQuaternionFromEuler(Euler=(X=30,Y=0,Z=-45), RotationOrder=ZXY).Result
    Weight=@n7.OutputParam
   )
   > n8 = &RigUnit_SphericalPoseReader(
    ActiveRegionScaleFactors=(PositiveWidth=0.5,NegativeWidth=0.5,PositiveHeight=0.5,NegativeHeight=0.5)
    ActiveRegionSize=0.1
    Debug=(bDrawDebug=false,bDraw2D=False,bDrawLocalAxes=False,DebugScale=25,DebugSegments=20,DebugThickness=0.25)
    DriverAxis=(X=0,Y=0,Z=1)
    DriverItem=(Type=Bone,Name="upperarm_r")
    FalloffRegionScaleFactors=(PositiveWidth=0.25,NegativeWidth=0.25,PositiveHeight=0.25,NegativeHeight=0.25)
    FalloffSize=0.4
    FlipHeightScaling=False
    FlipWidthScaling=False
    OptionalParentItem=(Type=Bone,Name="None")
    RotationOffset=(X=-10,Y=-60,Z=0)
   )
   > &RigUnit_ModifyTransforms(
    ItemToModify=((Item=(Type=Bone,Name="dyn_upperarm_twist_01_corr_r"),Transform=(Rotation=(X=0,Y=0,Z=0,W=1),Translation=(X=0,Y=0,Z=-2),Scale3D=(X=1,Y=1,Z=1))))
    ItemToModify.0.Transform.Rotation=&RigVMFunction_MathQuaternionFromEuler(Euler=(X=0,Y=0,Z=0), RotationOrder=ZXY).Result
    Weight=@n8.OutputParam
   )
   > n9 = &RigUnit_SphericalPoseReader(
    ActiveRegionScaleFactors=(PositiveWidth=1,NegativeWidth=1,PositiveHeight=1,NegativeHeight=1)
    ActiveRegionSize=0.1
    Debug=(bDrawDebug=true,bDraw2D=False,bDrawLocalAxes=False,DebugScale=25,DebugSegments=20,DebugThickness=0.25)
    DriverAxis=(X=-1,Y=0,Z=0)
    DriverItem=(Type=Bone,Name="upperarm_r")
    FalloffRegionScaleFactors=(PositiveWidth=0.33,NegativeWidth=0.33,PositiveHeight=0.33,NegativeHeight=0.33)
    FalloffSize=0.4
    FlipHeightScaling=False
    FlipWidthScaling=False
    OptionalParentItem=(Type=Bone,Name="None")
    RotationOffset=(X=0,Y=30,Z=-55)
   )
   Multiply = &RigVMFunction_MathDoubleMul(A=@n9.OutputParam, B=RigVMVariableNode() {variable=is_FE}.Value)
   > &RigUnit_ModifyTransforms(
    ItemToModify=((Item=(Type=Bone,Name="deform_upperarm_bk_base_r"),Transform=(Rotation=(X=0,Y=0,Z=0,W=1),Translation=(X=20,Y=15,Z=20),Scale3D=(X=1,Y=1,Z=1))))
    ItemToModify.0.Transform.Rotation=&RigVMFunction_MathQuaternionFromEuler(Euler=(X=0,Y=0,Z=0), RotationOrder=ZXY).Result
    Weight=@Multiply.Result
   )
   RigVMFunction_Sequence = &RigVMFunction_Sequence(ExecuteContext=@n1.B)
   n10 = &RigUnit_SphericalPoseReader(
    ActiveRegionScaleFactors=(PositiveWidth=1,NegativeWidth=1,PositiveHeight=1,NegativeHeight=1)
    ActiveRegionSize=0.1
    Debug=(bDrawDebug=true,bDraw2D=False,bDrawLocalAxes=False,DebugScale=25,DebugSegments=20,DebugThickness=0.25)
    DriverAxis=(X=-1,Y=0,Z=0)
    DriverItem=(Type=Bone,Name="ball_r")
    FalloffRegionScaleFactors=(PositiveWidth=1,NegativeWidth=1,PositiveHeight=1,NegativeHeight=1)
    FalloffSize=0.35
    FlipHeightScaling=False
    FlipWidthScaling=False
    OptionalParentItem=(Type=Bone,Name="None")
    RotationOffset=(X=90,Y=0,Z=30)
    ExecutePin=@RigVMFunction_Sequence.A
   )
   > &RigUnit_ModifyTransforms(
    ItemToModify=((Item=(Type=Bone,Name="dyn_butterfly_wing_base_r"),Transform=(Rotation=(X=0,Y=0,Z=0,W=1),Translation=(X=0,Y=-0.5,Z=0.5),Scale3D=(X=1,Y=1,Z=1))))
    ItemToModify.0.Transform.Rotation=&RigVMFunction_MathQuaternionFromEuler(Euler=(X=30,Y=0,Z=0), RotationOrder=ZXY).Result
    Weight=@n10.OutputParam
   )
   n11 = &RigUnit_SphericalPoseReader(
    ActiveRegionScaleFactors=(PositiveWidth=1,NegativeWidth=1,PositiveHeight=1,NegativeHeight=1)
    ActiveRegionSize=0.1
    Debug=(bDrawDebug=true,bDraw2D=False,bDrawLocalAxes=False,DebugScale=25,DebugSegments=20,DebugThickness=0.25)
    DriverAxis=(X=1,Y=0,Z=0)
    DriverItem=(Type=Bone,Name="ball_l")
    FalloffRegionScaleFactors=(PositiveWidth=1,NegativeWidth=1,PositiveHeight=1,NegativeHeight=1)
    FalloffSize=0.35
    FlipHeightScaling=False
    FlipWidthScaling=False
    OptionalParentItem=(Type=Bone,Name="None")
    RotationOffset=(X=-90,Y=0,Z=30)
    ExecutePin=@RigVMFunction_Sequence.B
   )
   > &RigUnit_ModifyTransforms(
    ItemToModify=((Item=(Type=Bone,Name="dyn_butterfly_wing_base_l"),Transform=(Rotation=(X=0,Y=0,Z=0,W=1),Translation=(X=0,Y=-1,Z=1),Scale3D=(X=1,Y=1,Z=1))))
    ItemToModify.0.Transform.Rotation=&RigVMFunction_MathQuaternionFromEuler(Euler=(X=30,Y=0,Z=0), RotationOrder=ZXY).Result
    Weight=@n11.OutputParam
   )
   RigVMFunction_Sequence_7 = RigVMAggregateNode(ExecuteContext=())
   n12 = &RigUnit_SphericalPoseReader(
    ActiveRegionScaleFactors=(PositiveWidth=0.5,NegativeWidth=0.5,PositiveHeight=0.5,NegativeHeight=0.5)
    ActiveRegionSize=0.1
    Debug=(bDrawDebug=true,bDraw2D=False,bDrawLocalAxes=False,DebugScale=25,DebugSegments=20,DebugThickness=0.25)
    DriverAxis=(X=-1,Y=0,Z=0)
    DriverItem=(Type=Bone,Name="hand_r")
    FalloffRegionScaleFactors=(PositiveWidth=0.5,NegativeWidth=1,PositiveHeight=1,NegativeHeight=1)
    FalloffSize=0.35
    FlipHeightScaling=False
    FlipWidthScaling=False
    OptionalParentItem=(Type=Bone,Name="None")
    RotationOffset=(X=20,Y=140,Z=0)
    ExecutePin=@RigVMFunction_Sequence_7.A
   )
   > &RigUnit_ModifyTransforms(
    ItemToModify=((Item=(Type=Bone,Name="dyn_bracelet_r"),Transform=(Rotation=(X=0,Y=0,Z=0,W=1),Translation=(X=1,Y=-0.25,Z=0),Scale3D=(X=1,Y=1,Z=1))))
    ItemToModify.0.Transform.Rotation=&RigVMFunction_MathQuaternionFromEuler(Euler=(X=0,Y=0,Z=-10), RotationOrder=ZXY).Result
    Weight=@n12.OutputParam
   )
   n13 = &RigUnit_SphericalPoseReader(
    ActiveRegionScaleFactors=(PositiveWidth=0.5,NegativeWidth=0.5,PositiveHeight=0.5,NegativeHeight=0.5)
    ActiveRegionSize=0.1
    Debug=(bDrawDebug=true,bDraw2D=False,bDrawLocalAxes=False,DebugScale=25,DebugSegments=20,DebugThickness=0.25)
    DriverAxis=(X=-1,Y=0,Z=0)
    DriverItem=(Type=Bone,Name="hand_r")
    FalloffRegionScaleFactors=(PositiveWidth=1,NegativeWidth=1,PositiveHeight=1,NegativeHeight=1)
    FalloffSize=0.35
    FlipHeightScaling=False
    FlipWidthScaling=False
    OptionalParentItem=(Type=Bone,Name="None")
    RotationOffset=(X=-25,Y=50,Z=0)
    ExecutePin=@RigVMFunction_Sequence_7.B
   )
   > &RigUnit_ModifyTransforms(
    ItemToModify=((Item=(Type=Bone,Name="dyn_bracelet_r"),Transform=(Rotation=(X=0,Y=0,Z=0,W=1),Translation=(X=0,Y=0,Z=0),Scale3D=(X=1,Y=1,Z=1))))
    ItemToModify.0.Transform.Rotation=&RigVMFunction_MathQuaternionFromEuler(Euler=(X=0,Y=0,Z=0), RotationOrder=ZXY).Result
    Weight=@n13.OutputParam
   )
   n14 = &RigUnit_SphericalPoseReader(
    ActiveRegionScaleFactors=(PositiveWidth=0.5,NegativeWidth=0.5,PositiveHeight=0.5,NegativeHeight=0.5)
    ActiveRegionSize=0.1
    Debug=(bDrawDebug=true,bDraw2D=False,bDrawLocalAxes=False,DebugScale=25,DebugSegments=20,DebugThickness=0.25)
    DriverAxis=(X=-1,Y=0,Z=0)
    DriverItem=(Type=Bone,Name="hand_r")
    FalloffRegionScaleFactors=(PositiveWidth=0.5,NegativeWidth=1,PositiveHeight=1,NegativeHeight=0.5)
    FalloffSize=0.25
    FlipHeightScaling=False
    FlipWidthScaling=False
    OptionalParentItem=(Type=Bone,Name="None")
    RotationOffset=(X=25,Y=75,Z=0)
    ExecutePin=@RigVMFunction_Sequence_7.C
   )
   > &RigUnit_ModifyTransforms(
    ItemToModify=((Item=(Type=Bone,Name="dyn_bracelet_r"),Transform=(Rotation=(X=0,Y=0,Z=0,W=1),Translation=(X=1,Y=0.2,Z=0),Scale3D=(X=1,Y=1,Z=1))))
    ItemToModify.0.Transform.Rotation=&RigVMFunction_MathQuaternionFromEuler(Euler=(X=0,Y=0,Z=5), RotationOrder=ZXY).Result
    Weight=@n14.OutputParam
   )
   n15 = &RigUnit_SphericalPoseReader(
    ActiveRegionScaleFactors=(PositiveWidth=0.5,NegativeWidth=0.5,PositiveHeight=0.5,NegativeHeight=0.5)
    ActiveRegionSize=0.1
    Debug=(bDrawDebug=true,bDraw2D=False,bDrawLocalAxes=False,DebugScale=25,DebugSegments=20,DebugThickness=0.25)
    DriverAxis=(X=-1,Y=0,Z=0)
    DriverItem=(Type=Bone,Name="upperarm_r")
    ExecutePin=()
    FalloffRegionScaleFactors=(PositiveWidth=0.25,NegativeWidth=0.25,PositiveHeight=0.25,NegativeHeight=0.25)
    FalloffSize=0.4
    FlipHeightScaling=False
    FlipWidthScaling=False
    OptionalParentItem=(Type=Bone,Name="None")
    RotationOffset=(X=15,Y=30,Z=0)
   )
   > &RigUnit_ModifyTransforms(
    ItemToModify=((Item=(Type=Bone,Name="dyn_upperarm_twist_01_corr_r"),Transform=(Rotation=(X=0,Y=0,Z=0,W=1),Translation=(X=0,Y=-2,Z=0),Scale3D=(X=1,Y=1,Z=1))))
    ItemToModify.0.Transform.Rotation=&RigVMFunction_MathQuaternionFromEuler(Euler=(X=20,Y=-20,Z=30), RotationOrder=ZXY).Result
    Weight=@n15.OutputParam
   )
   RigVMRerouteNode(Value=&RigVMFunction_MathDoubleMul(B=1, A=@n14.OutputParam).Result)
  ]}
]
hierarchy [
 root {loc=<0 0 0>, rot=<0 0 0 1>}
 pelvis {parent=root, loc=<0 -0.5301 98.0571>, rot=<0.0127 0.707 -0.0127 -0.707>}
 spine_01 {parent=pelvis, loc=<6.1208 -0.9328 0>, rot=<0 0 0.034 -0.9994>}
 spine_02 {parent=spine_01, loc=<6.5997 -0.2095 0>, rot=<0 0 0.0238 -0.9997>}
 spine_03 {parent=spine_02, loc=<6.4914 -0.1237 0>, rot=<0 0 0.0149 -0.9999>}
 spine_04 {parent=spine_03, loc=<6.9598 0.0794 0>, rot=<0 0 -0.0132 -0.9999>}
 spine_05 {parent=spine_04, loc=<6.9987 0 0>, rot=<0 0 -0.0015 -1>}
 neck_01 {parent=spine_05, loc=<11.3138 0.1529 0>, rot=<0 0 -0.2003 -0.9797>}
 neck_02 {parent=neck_01, loc=<4.5627 0 0>, rot=<0 0 0.0504 -0.9987>}
 head {parent=neck_02, loc=<5.3914 0 0>, rot=<0 0 0.115 -0.9934>}
 dyn_hair_tail_base {parent=head, loc=<16.4031 -12.0456 1.45>, rot=<0.0997 -0.7594 -0.6282 0.1369>}
 dyn_hair_tail_1 {parent=dyn_hair_tail_base, loc=<0 0 0>, rot=<0 0 0 1>}
 dyn_hair_tail_2 {parent=dyn_hair_tail_1, loc=<9.364 0 0>, rot=<0 0.187 0 -0.9824>}
 dyn_hair_tail_3 {parent=dyn_hair_tail_2, loc=<12.0222 0 0>, rot=<0.4081 -0.1801 0.0825 -0.8912>}
 dyn_hair_tail_4 {parent=dyn_hair_tail_3, loc=<9.7604 0 0>, rot=<0.1092 0.3171 -0.0368 -0.9414>}
 dyn_earring_base_l {parent=head, loc=<0.7306 0.4972 -6.8542>, rot=<0.0225 -0.9583 -0.2837 -0.0245>}
 dyn_earring_l {parent=dyn_earring_base_l, loc=<0 0 0>, rot=<0 0 0 1>}
 dyn_earring_base_r {parent=head, loc=<0.73 0.4869 6.8385>, rot=<0.0225 -0.9583 0.2837 0.0245>}
 dyn_earring_r {parent=dyn_earring_base_r, loc=<0 0 0>, rot=<0 0 0 1>}
 dyn_lock_of_hair_3_base_r {parent=head, loc=<9.1814 6.9531 7.6983>, rot=<0.2467 -0.936 0.0392 0.248>}
 dyn_lock_of_hair_3_1_r {parent=dyn_lock_of_hair_3_base_r, loc=<0 0 0>, rot=<0 0 0 1>}
 dyn_lock_of_hair_3_2_r {parent=dyn_lock_of_hair_3_1_r, loc=<4.0325 0 0>, rot=<0 0.2362 0 -0.9717>}
 dyn_lock_of_hair_2_base_r {parent=head, loc=<12.4684 9.6489 2.6788>, rot=<0.2247 -0.856 0.4298 0.179>}
 dyn_lock_of_hair_2_1_r {parent=dyn_lock_of_hair_2_base_r, loc=<0 0 0>, rot=<0 0 0 1>}
 dyn_lock_of_hair_2_2_r {parent=dyn_lock_of_hair_2_1_r, loc=<7.8595 0 0>, rot=<0 0.0995 0 -0.995>}
 dyn_lock_of_hair_2_3_r {parent=dyn_lock_of_hair_2_2_r, loc=<7.0175 0 0>, rot=<-0.089 0.0394 -0.4025 -0.9102>}
 dyn_lock_of_hair_1_base_r {parent=head, loc=<11.1113 10.5664 0.9769>, rot=<-0.1158 -0.9712 0.0145 0.2078>}
 dyn_lock_of_hair_1_1_r {parent=dyn_lock_of_hair_1_base_r, loc=<0 0 0>, rot=<0 0 0 1>}
 dyn_lock_of_hair_3_base_l {parent=head, loc=<1.6076 -1.7793 -6.4374>, rot=<-0.011 -0.8533 -0.5194 0.0447>}
 dyn_lock_of_hair_3_1_l {parent=dyn_lock_of_hair_3_base_l, loc=<0 0 0>, rot=<0 0 0 1>}
 dyn_lock_of_hair_3_2_l {parent=dyn_lock_of_hair_3_1_l, loc=<4.9804 0 0>, rot=<0 0.1939 0 -0.981>}
 dyn_lock_of_hair_3_3_l {parent=dyn_lock_of_hair_3_2_l, loc=<6.091 0 0>, rot=<-0.2072 0.098 -0.4161 -0.88>}
 dyn_lock_of_hair_2_base_l {parent=head, loc=<11.427 10.9514 -2.3397>, rot=<-0.0438 -0.9465 0.3027 -0.1033>}
 dyn_lock_of_hair_2_1_l {parent=dyn_lock_of_hair_2_base_l, loc=<0 0 0>, rot=<0 0 0 1>}
 dyn_lock_of_hair_2_2_l {parent=dyn_lock_of_hair_2_1_l, loc=<6.7131 0.0407 0>, rot=<-0.0016 0.0454 -0.0344 -0.9984>}
 dyn_lock_of_hair_2_3_l {parent=dyn_lock_of_hair_2_2_l, loc=<6.5957 0.4218 0>, rot=<-0.0395 -0.0174 0.402 -0.9146>}
 dyn_lock_of_hair_1_base_l {parent=head, loc=<11.0507 10.4149 -1.0799>, rot=<-0.2208 -0.9692 -0.0252 -0.1066>}
 dyn_lock_of_hair_1_l {parent=dyn_lock_of_hair_1_base_l, loc=<0 0 0>, rot=<0 0 0 1>}
 dyn_necklace_base {parent=neck_01, loc=<0 0 0>, rot=<-0.5679 -0.8228 -0.0069 -0.0237>}
 dyn_necklace_1 {parent=dyn_necklace_base, loc=<0 0 0>, rot=<0 0 0 1>}
 clavicle_l {parent=spine_05, loc=<5.4831 -0.1053 -3.2946>, rot=<0.0465 -0.7289 -0.009 -0.6829>}
 upperarm_l {parent=clavicle_l, loc=<9.8658 -1.2136 -0.4105>, rot=<0.0164 -0.3245 -0.0269 -0.9454>}
 lowerarm_l {parent=upperarm_l, loc=<26.465 -0.8345 0.259>, rot=<0.0036 0.0161 -0.2298 -0.9731>}
 hand_l {parent=lowerarm_l, loc=<22.7355 -0.3203 -0.6105>, rot=<-0.6784 -0.0223 0.0735 -0.7307>}
 thumb_01_l {parent=hand_l, loc=<1.3118 -0.3191 -2.459>, rot=<0.6546 -0.4042 0.0664 -0.6354>}
 thumb_02_l {parent=thumb_01_l, loc=<5.9336 -0.0448 0.0401>, rot=<-0.0037 0.0213 0.298 -0.9543>}
 thumb_03_l {parent=thumb_02_l, loc=<2.2223 0.3682 -0.0907>, rot=<0.0118 -0.013 0.0066 -0.9998>}
 index_metacarpal_l {parent=hand_l, loc=<1.6535 0.0183 -2.96>, rot=<0.0158 -0.0124 -0.0173 -0.9996>}
 index_01_l {parent=index_metacarpal_l, loc=<8.3008 0.0357 -0.0524>, rot=<-0.0044 0.0485 0.0924 -0.9945>}
 index_02_l {parent=index_01_l, loc=<3.3587 -0.3004 0.0452>, rot=<-0.002 0.0027 0.3225 -0.9466>}
 index_03_l {parent=index_02_l, loc=<2.5578 -0.123 -0.0252>, rot=<-0.002 0.0067 0.2558 -0.9667>}
 middle_metacarpal_l {parent=hand_l, loc=<2.1856 -0.1725 -0.5984>, rot=<0.006 0.0168 -0.0329 -0.9993>}
 middle_01_l {parent=middle_metacarpal_l, loc=<7.3275 -0.2447 0.0129>, rot=<-0.0208 0.0042 0.0858 -0.9961>}
 middle_02_l {parent=middle_01_l, loc=<4.0123 0.1401 0.1175>, rot=<-0.0218 0.0258 0.3512 -0.9357>}
 middle_03_l {parent=middle_02_l, loc=<2.6077 0.0534 0.1373>, rot=<0.0082 0.0188 0.1708 -0.9851>}
 ring_metacarpal_l {parent=hand_l, loc=<2.2074 -0.3625 1.3638>, rot=<-0.0284 0.0722 -0.0458 -0.9959>}
 ring_01_l {parent=ring_metacarpal_l, loc=<7.0295 -0.6197 -0.1446>, rot=<-0.0489 -0.0252 0.1892 -0.9804>}
 ring_02_l {parent=ring_01_l, loc=<3.6174 0.4647 -0.1027>, rot=<0 0 0.2806 -0.9598>}
 ring_03_l {parent=ring_02_l, loc=<2.4152 0.114 0.0199>, rot=<0 0 0.1189 -0.9929>}
 pinky_metacarpal_l {parent=hand_l, loc=<1.7616 -0.9753 3.3814>, rot=<-0.0397 0.0814 -0.069 -0.9935>}
 pinky_01_l {parent=pinky_metacarpal_l, loc=<7.1709 -0.4831 -0.1094>, rot=<-0.0654 -0.0035 0.2242 -0.9723>}
 pinky_02_l {parent=pinky_01_l, loc=<2.8274 0.0893 -0.1043>, rot=<0.0037 0.0011 0.2854 -0.9584>}
 pinky_03_l {parent=pinky_02_l, loc=<2.2046 0.0307 -0.0078>, rot=<-0.0002 0.0098 0.1445 -0.9895>}
 deform_wrist_dn_base_l {parent=hand_l, loc=<0 -1.5326 0>, rot=<0 0 0 -1>}
 deform_wrist_dn_l {parent=deform_wrist_dn_base_l, loc=<0 0 0>, rot=<0 0 0 1>}
 deform_wrist_up_base_l {parent=hand_l, loc=<0 2.3009 0>, rot=<0 0 0 -1>}
 deform_wrist_up_l {parent=deform_wrist_up_base_l, loc=<0 0 0>, rot=<0 0 0 1>}
 hand_attach_l {parent=hand_l, loc=<8.5 -3 0>, rot=<0.7071 0.7071 0 0>}
 weapon_l {parent=hand_l, loc=<-0.1974 -3.4111 0.3811>, rot=<0 0 0 1>}
 lowerarm_twist_02_l {parent=lowerarm_l, loc=<10.6707 0 0>, rot=<0 0 0 1>}
 lowerarm_twist_01_l {parent=lowerarm_l, loc=<18.316 0 0>, rot=<0 0 0 1>}
 dyn_bracelet_2_l {parent=lowerarm_l, loc=<21.1495 -0.0539 -0.4773>, rot=<-0.0401 -0.0606 -0.018 -0.9972>}
 dyn_bracelet_1_l {parent=lowerarm_l, loc=<18.0893 -0.1796 -0.1088>, rot=<-0.0401 -0.0606 -0.018 -0.9972>}
 deform_elbow_in_base_l {parent=lowerarm_l, loc=<0.1834 3.6062 -0.2572>, rot=<0 0 0 -1>}
 deform_elbow_in_l {parent=deform_elbow_in_base_l, loc=<0 0 0>, rot=<0 0 0 1>}
 deform_elbow_out_base_l {parent=lowerarm_l, loc=<0 -4 0>, rot=<0 0 0 1>}
 deform_elbow_out_l {parent=deform_elbow_out_base_l, loc=<0 0 0>, rot=<0 0 0 1>}
 upperarm_twist_01_l {parent=upperarm_l, loc=<7.2135 -0.1134 0>, rot=<0 0 0 1>}
 deform_upperarm_bk_base_l {parent=upperarm_twist_01_l, loc=<-6.4053 -4.3746 0.314>, rot=<0 0 0 -1>}
 deform_upperarm_bk_l {parent=deform_upperarm_bk_base_l, loc=<0 0 0>, rot=<0 0 0 1>}
 deform_upperarm_fr_base_l {parent=upperarm_twist_01_l, loc=<-5.6312 5.092 0.3354>, rot=<0 0 0 -1>}
 deform_upperarm_fr_l {parent=deform_upperarm_fr_base_l, loc=<0 0 0>, rot=<0 0 0 1>}
 deform_upperarm_up_base_l {parent=upperarm_twist_01_l, loc=<-8.891 0.4825 3.9517>, rot=<0 0 0 -1>}
 deform_upperarm_up_l {parent=deform_upperarm_up_base_l, loc=<0 0 0>, rot=<0 0 0 1>}
 dyn_pendant_1_base_l {parent=upperarm_twist_01_l, loc=<2.1295 3.597 1.3409>, rot=<0.4522 -0.1857 -0.2092 -0.8469>}
 dyn_pendant_1_l {parent=dyn_pendant_1_base_l, loc=<0 0 0>, rot=<0 0 0 1>}
 dyn_pendant_2_base_l {parent=upperarm_twist_01_l, loc=<3.2417 -0.7613 3.2865>, rot=<-0.0071 -0.0736 0.0074 -0.9972>}
 dyn_pendant_2_l {parent=dyn_pendant_2_base_l, loc=<0 0 0>, rot=<0 0 0 1>}
 dyn_pendant_3_base_l {parent=upperarm_twist_01_l, loc=<1.4561 -5.2478 -0.2419>, rot=<-0.6584 -0.1594 0.2709 -0.6839>}
 dyn_pendant_3_l {parent=dyn_pendant_3_base_l, loc=<0 0 0>, rot=<0 0 0 1>}
 upperarm_twist_02_l {parent=upperarm_l, loc=<19.7998 -0.1134 0>, rot=<0 0 0 1>}
 clavicle_r {parent=spine_05, loc=<5.4831 -0.1053 3.2946>, rot=<0.7289 0.0465 -0.6829 0.009>}
 upperarm_r {parent=clavicle_r, loc=<-9.8658 1.2136 0.4105>, rot=<0.0164 -0.3245 -0.0269 -0.9454>}
 lowerarm_r {parent=upperarm_r, loc=<-26.465 0.8345 -0.259>, rot=<0.0036 0.0161 -0.2298 -0.9731>}
 hand_r {parent=lowerarm_r, loc=<-22.7355 0.3203 0.6106>, rot=<-0.6784 -0.0223 0.0735 -0.7307>}
 thumb_01_r {parent=hand_r, loc=<-1.3118 0.3191 2.459>, rot=<0.6546 -0.4042 0.0664 -0.6354>}
 thumb_02_r {parent=thumb_01_r, loc=<-5.9336 0.0448 -0.0401>, rot=<-0.0037 0.0213 0.298 -0.9543>}
 thumb_03_r {parent=thumb_02_r, loc=<-2.2223 -0.3682 0.0907>, rot=<0.0118 -0.013 0.0066 -0.9998>}
 index_metacarpal_r {parent=hand_r, loc=<-1.6535 -0.0183 2.96>, rot=<0.0158 -0.0124 -0.0173 -0.9996>}
 index_01_r {parent=index_metacarpal_r, loc=<-8.3008 -0.0357 0.0524>, rot=<-0.0044 0.0485 0.0924 -0.9945>}
 index_02_r {parent=index_01_r, loc=<-3.3587 0.3004 -0.0452>, rot=<-0.002 0.0027 0.3225 -0.9466>}
 index_03_r {parent=index_02_r, loc=<-2.5578 0.123 0.0252>, rot=<-0.002 0.0067 0.2558 -0.9667>}
 middle_metacarpal_r {parent=hand_r, loc=<-2.1856 0.1725 0.5984>, rot=<0.006 0.0168 -0.0329 -0.9993>}
 middle_01_r {parent=middle_metacarpal_r, loc=<-7.3275 0.2447 -0.0129>, rot=<-0.0208 0.0042 0.0858 -0.9961>}
 middle_02_r {parent=middle_01_r, loc=<-4.0123 -0.1401 -0.1175>, rot=<-0.0218 0.0258 0.3512 -0.9357>}
 middle_03_r {parent=middle_02_r, loc=<-2.6077 -0.0534 -0.1373>, rot=<0.0082 0.0188 0.1708 -0.9851>}
 ring_metacarpal_r {parent=hand_r, loc=<-2.2074 0.3625 -1.3638>, rot=<-0.0284 0.0722 -0.0458 -0.9959>}
 ring_01_r {parent=ring_metacarpal_r, loc=<-7.0295 0.6197 0.1446>, rot=<-0.0489 -0.0252 0.1892 -0.9804>}
 ring_02_r {parent=ring_01_r, loc=<-3.6174 -0.4647 0.1027>, rot=<0 0 0.2806 -0.9598>}
 ring_03_r {parent=ring_02_r, loc=<-2.4152 -0.114 -0.0199>, rot=<0 0 0.1189 -0.9929>}
 pinky_metacarpal_r {parent=hand_r, loc=<-1.7616 0.9753 -3.3814>, rot=<-0.0397 0.0814 -0.069 -0.9935>}
 pinky_01_r {parent=pinky_metacarpal_r, loc=<-7.1709 0.4831 0.1094>, rot=<-0.0654 -0.0035 0.2242 -0.9723>}
 pinky_02_r {parent=pinky_01_r, loc=<-2.8274 -0.0893 0.1043>, rot=<0.0037 0.0011 0.2854 -0.9584>}
 pinky_03_r {parent=pinky_02_r, loc=<-2.2046 -0.0307 0.0078>, rot=<-0.0002 0.0098 0.1445 -0.9895>}
 deform_wrist_dn_base_r {parent=hand_r, loc=<-0.0001 1.5328 0>, rot=<0 0 0 -1>}
 deform_wrist_dn_r {parent=deform_wrist_dn_base_r, loc=<0 0 0>, rot=<0 0 0 1>}
 deform_wrist_up_base_r {parent=hand_r, loc=<-0.0002 -2.3007 -0.0001>, rot=<0 0 0 -1>}
 deform_wrist_up_r {parent=deform_wrist_up_base_r, loc=<0 0 0>, rot=<0 0 0 1>}
 weapon_r {parent=hand_r, loc=<0.1974 3.4111 -0.3811>, rot=<0 0 0 1>}
 hand_attach_r {parent=hand_r, loc=<-8.5001 2.9999 0>, rot=<0 0 -0.7071 -0.7071>}
 lowerarm_twist_02_r {parent=lowerarm_r, loc=<-10.6707 0 0>, rot=<0 0 0 -1>}
 lowerarm_twist_01_r {parent=lowerarm_r, loc=<-18.316 0 0>, rot=<0 0 0 -1>}
 dyn_bracelet_r {parent=lowerarm_r, loc=<-19.5374 0.1011 -0.0115>, rot=<0 0 0 1>}
 deform_elbow_in_base_r {parent=lowerarm_r, loc=<-0.1834 -3.6062 0.2572>, rot=<0 0 0 -1>}
 deform_elbow_in_r {parent=deform_elbow_in_base_r, loc=<0 0 0>, rot=<0 0 0 1>}
 deform_elbow_out_base_r {parent=lowerarm_r, loc=<0 4 -0.0001>, rot=<0 0 0 -1>}
 deform_elbow_out_r {parent=deform_elbow_out_base_r, loc=<0 0 0>, rot=<0 0 0 1>}
 upperarm_twist_01_r {parent=upperarm_r, loc=<-7.2135 0.1134 0>, rot=<0 0 0 -1>}
 deform_upperarm_bk_base_r {parent=upperarm_twist_01_r, loc=<6.4053 4.3746 -0.314>, rot=<0 0 0 -1>}
 deform_upperarm_bk_r {parent=deform_upperarm_bk_base_r, loc=<0 0 0>, rot=<0 0 0 1>}
 deform_upperarm_fr_base_r {parent=upperarm_twist_01_r, loc=<5.6312 -5.092 -0.3354>, rot=<0 0 0 -1>}
 deform_upperarm_fr_r {parent=deform_upperarm_fr_base_r, loc=<0 0 0>, rot=<0 0 0 1>}
 deform_upperarm_up_base_r {parent=upperarm_twist_01_r, loc=<8.891 -0.4825 -3.9517>, rot=<0 0 0 -1>}
 deform_upperarm_up_r {parent=deform_upperarm_up_base_r, loc=<0 0 0>, rot=<0 0 0 1>}
 dyn_pendant_1_base_r {parent=upperarm_twist_01_r, loc=<-2.1295 -3.597 -1.3409>, rot=<0.1857 0.4522 0.8469 -0.2092>}
 dyn_pendant_1_r {parent=dyn_pendant_1_base_r, loc=<0 0 0>, rot=<0 0 0 1>}
 dyn_pendant_2_base_r {parent=upperarm_twist_01_r, loc=<-3.2417 0.7613 -3.2865>, rot=<0.0736 -0.0071 0.9972 0.0074>}
 dyn_pendant_2_r {parent=dyn_pendant_2_base_r, loc=<0 0 0>, rot=<0 0 0 1>}
 dyn_pendant_3_base_r {parent=upperarm_twist_01_r, loc=<-1.4561 5.2478 0.2419>, rot=<0.1594 -0.6584 0.6839 0.2709>}
 dyn_pendant_3_r {parent=dyn_pendant_3_base_r, loc=<0 0 0>, rot=<0 0 0 1>}
 dyn_upperarm_twist_01_corr_r {parent=upperarm_twist_01_r, loc=<0 0 0>, rot=<0 0 0 1>}
 upperarm_twist_02_r {parent=upperarm_r, loc=<-19.7998 0.1134 0>, rot=<0 0 0 -1>}
 attach_backpack {parent=spine_05, loc=<-0.123 -13.473 0>, rot=<-0.0093 -0.707 0.0093 -0.707>}
 Backpack_BR {parent=attach_backpack, type=Null, loc=<0 3.5 0>, rot=<0.0436 0 0 0.999>}
 attach_cape {parent=spine_05, loc=<9.017 -11.714 0>, rot=<0.0616 -0.7044 -0.0616 -0.7044>}
 cape {parent=attach_cape, type=Null, loc=<0 5.0416 1.3509>, rot=<0.1305 0 0 0.9914>}
 dyn_spine_05_corr {parent=spine_05, loc=<0.41 1.7023 11.3795>, rot=<0 0 -0.0401 -0.9992>}
 deform_pec_aim_base_l {parent=spine_04, loc=<3.2456 9.3309 -1.2237>, rot=<0.0603 -0.4033 0.3318 -0.8507>}
 deform_pec_aim_l {parent=deform_pec_aim_base_l, loc=<0 0 0>, rot=<0 0 0 1>}
 deform_pec_base_l {parent=deform_pec_aim_l, loc=<9.3665 0 0>, rot=<0 0 0 1>}
 deform_pec_l {parent=deform_pec_base_l, loc=<0 0 0>, rot=<0 0 0 1>}
 deform_pec_aim_base_r {parent=spine_04, loc=<3.2452 9.3309 1.2237>, rot=<-0.4033 -0.0603 0.8507 0.3318>}
 deform_pec_aim_r {parent=deform_pec_aim_base_r, loc=<0 0 0>, rot=<0 0 0 1>}
 deform_pec_base_r {parent=deform_pec_aim_r, loc=<-9.3667 -0.0003 -0.0004>, rot=<0 0 0 1>}
 deform_pec_r {parent=deform_pec_base_r, loc=<0 0 0>, rot=<0 0 0 1>}
 dyn_bow {parent=spine_03, loc=<4.2428 -6.7497 -0.1062>, rot=<0.026 -0.9993 -0.0021 -0.0254>}
 thigh_l {parent=pelvis, loc=<-1.0855 -0.2109 -8.7053>, rot=<0.0009 0.0355 0.0155 -0.9993>}
 calf_l {parent=thigh_l, loc=<-42.5284 -0.0019 -0.7297>, rot=<0.0006 -0.0161 -0.0391 -0.9991>}
 foot_l {parent=calf_l, loc=<-46.7401 -2.5014 -0.592>, rot=<0.0427 -0.0238 0.0353 -0.9982>}
 ball_l {parent=foot_l, loc=<-3.5658 10.8066 -0.1544>, rot=<0.0001 -0.0001 -0.7063 -0.7079>}
 dyn_butterfly_wing_base_l {parent=foot_l, loc=<0.3705 10.8279 0.0621>, rot=<-0.0357 -0.7045 -0.0284 -0.7082>}
 dyn_butterfly_wing_1_l {parent=dyn_butterfly_wing_base_l, loc=<0.6204 -0.1642 0.0344>, rot=<0.0005 -0.0075 0.0701 -0.9975>}
 dyn_butterfly_wing_2_l {parent=dyn_butterfly_wing_base_l, loc=<-0.7044 0.0557 0.0766>, rot=<-0.0945 -0.9952 0.0023 0.0241>}
 calf_twist_02_l {parent=calf_l, loc=<-19.1546 -1.7943 -0.2638>, rot=<0.0237 0.0095 -0.0035 -0.9997>}
 calf_twist_01_l {parent=calf_l, loc=<-35.8483 -1.9197 -0.5794>, rot=<0.0237 0.0095 -0.0035 -0.9997>}
 deform_knee_in_base_l {parent=calf_l, loc=<0.8031 -5.0231 0.2456>, rot=<0 0 0 -1>}
 deform_knee_in_l {parent=deform_knee_in_base_l, loc=<0 0 0>, rot=<0 0 0 1>}
 deform_knee_out_base_l {parent=calf_l, loc=<0 4 0>, rot=<0 0 0 -1>}
 deform_knee_out_l {parent=deform_knee_out_base_l, loc=<0 0 0>, rot=<0 0 0 1>}
 thigh_twist_01_l {parent=thigh_l, loc=<-27.952 0 0>, rot=<0.0013 0 0 -1>}
 deform_groin_base_l {parent=thigh_l, loc=<0.4801 5.6776 -0.2486>, rot=<0 0 0 -1>}
 deform_groin_l {parent=deform_groin_base_l, loc=<0 0 0>, rot=<0 0 0 1>}
 thigh_r {parent=pelvis, loc=<-1.0855 -0.2109 8.7053>, rot=<0.0355 -0.0009 0.9993 0.0155>}
 calf_r {parent=thigh_r, loc=<42.5284 0.0019 0.7297>, rot=<0.0006 -0.0161 -0.0391 -0.9991>}
 foot_r {parent=calf_r, loc=<46.7401 2.5014 0.592>, rot=<0.0427 -0.0238 0.0353 -0.9982>}
 ball_r {parent=foot_r, loc=<3.5658 -10.8066 0.1544>, rot=<0.0001 -0.0001 -0.7063 -0.7079>}
 dyn_butterfly_wing_base_r {parent=foot_r, loc=<-0.3705 -10.828 -0.0621>, rot=<0.7082 0.0284 -0.7045 -0.0357>}
 dyn_butterfly_wing_1_r {parent=dyn_butterfly_wing_base_r, loc=<0.7044 0.0556 0.0765>, rot=<0.0023 0.0241 -0.0945 -0.9952>}
 dyn_butterfly_wing_2_r {parent=dyn_butterfly_wing_base_r, loc=<-0.6204 -0.1643 0.0344>, rot=<0.0701 -0.9975 0.0005 -0.0075>}
 calf_twist_02_r {parent=calf_r, loc=<19.1546 1.7943 0.2638>, rot=<0.0237 0.0095 -0.0035 -0.9997>}
 calf_twist_01_r {parent=calf_r, loc=<35.8483 1.9197 0.5794>, rot=<0.0237 0.0095 -0.0035 -0.9997>}
 dyn_ankle_bracelet_r {parent=calf_r, loc=<37.1047 2.1909 0.6735>, rot=<0 0 0 1>}
 deform_knee_in_base_r {parent=calf_r, loc=<-0.8031 5.0231 -0.2457>, rot=<0 0 0 -1>}
 deform_knee_in_r {parent=deform_knee_in_base_r, loc=<0 0 0>, rot=<0 0 0 1>}
 deform_knee_out_base_r {parent=calf_r, loc=<0 -4 0>, rot=<0 0 0 -1>}
 deform_knee_out_r {parent=deform_knee_out_base_r, loc=<0 0 0>, rot=<0 0 0 1>}
 thigh_twist_01_r {parent=thigh_r, loc=<27.952 0 0>, rot=<0.0013 0 0 -1>}
 deform_groin_base_r {parent=thigh_r, loc=<-0.4801 -5.6776 0.2486>, rot=<0 0 0 -1>}
 deform_groin_r {parent=deform_groin_base_r, loc=<0 0 0>, rot=<0 0 0 1>}
 dyn_boba_drink_base {parent=pelvis, loc=<6.8994 -7.858 9.4265>, rot=<0.1877 -0.7893 -0.4434 0.3809>}
 dyn_boba_drink_1 {parent=dyn_boba_drink_base, loc=<0 0 0>, rot=<0 0 0 1>}
 dyn_boba_drink_2 {parent=dyn_boba_drink_1, loc=<6.799 0 0>, rot=<0 0.2394 0 -0.9709>}
 dyn_boba_drink_3 {parent=dyn_boba_drink_2, loc=<5.5272 0 0>, rot=<0.2496 0.1028 -0.0267 -0.9625>}
 deform_glute_base_l {parent=pelvis, loc=<-7.3466 -9.0213 -9.1245>, rot=<0.0172 0.031 -0.4477 -0.8935>}
 deform_glute_l {parent=deform_glute_base_l, loc=<0 0 0>, rot=<0 0 0 1>}
 deform_glute_base_r {parent=pelvis, loc=<-7.3464 -9.0213 9.1245>, rot=<0.031 -0.0172 0.8935 -0.4477>}
 deform_glute_r {parent=deform_glute_base_r, loc=<0 0 0>, rot=<0 0 0 1>}
]
variables [
 is_FE {type=double, default=1}
 is_crouch {type=double, default=0}
 is_emote {type=double, default=1}
 is_skydive {type=double, default=1}
]
