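# Physics asset for the F_MED_VibrantShell female body cosmetic, covering its dangling boba drink and bow bones.
# A four-body capsule chain runs from the drink base, with swing limits loosening toward the tip, while the bow is constrained to spine_03.
# Capsules on the pelvis, spine_03 and right leg only form the body shapes.
# An oversized root capsule centered far below the origin puts a broad, near-flat surface around floor level.
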PhysicsAsset /BRCosmetics/Characters/Player/Female/Medium/Bodies/F_MED_VibrantShell/Meshes/F_MED_VibrantShell_Physics
bodies {
 calf_r [{
   shapeType=Capsule
   length=37.338
   radius=5.5145
   shapeName=calf_r_capsule
   center=<21.9386 2.0995 0.2031>
   rotation=<rot 0.1548 86.072 -88.2302>
  }]
 dyn_boba_drink_1 [{
   shapeType=Capsule
   length=4.0493
   radius=1.8655
   shapeName=dyn_boba_drink_1_capsule
   center=<2.9608 -0.048 -0.2042>
   rotation=<rot 0 90.7953 -78.3879>
  }]
 dyn_boba_drink_2 [{
   shapeType=Capsule
   length=3.965
   radius=3.6676
   shapeName=dyn_boba_drink_2_capsule
   center=<3.2819 -0.456 0.1283>
   rotation=<rot 24.2945 -166.1607 -90>
  }]
 dyn_boba_drink_3 [{
   shapeType=Capsule
   length=8.649
   radius=4.9276
   shapeName=dyn_boba_drink_3_capsule
   center=<5.2465 -0.014 0.0358>
   rotation=<rot 0 90.3869 -89.3941>
  }]
 dyn_boba_drink_base [{
   shapeType=Capsule
   length=0
   radius=1.5374
   shapeName=dyn_boba_drink_base_capsule
   center=<-0.0927 -0.014 -0.0612>
   rotation=<rot -0.9851 2.3863 4.1487>
  }]
 dyn_bow [{
   shapeType=Capsule
   length=4.8776
   radius=2.5818
   shapeName=dyn_bow_capsule
   center=<4.0379 -0.2398 0.3317>
   rotation=<rot 0 89.8727 -84.6107>
  }]
 foot_r [{
   shapeType=Capsule
   length=12.7112
   radius=6.2382
   shapeName=foot_r_capsule
   center=<3.9669 -7.0951 -0.1711>
   rotation=<rot 0 8.2869 -89.4704>
  }]
 pelvis [{shapeType=Capsule, length=13.414, radius=8.6182, shapeName=pelvis_capsule, center=<-4.4857 -3.1415 -0.002>, rotation=<rot 0 87.4621 0.2346>}]
 root [{shapeType=Capsule, length=5.25, radius=203, shapeName=root_capsule, center=<0 0 -205.096>, rotation=<rot 0 -90 0>}]
 spine_03 [{
   shapeType=Capsule
   length=5.3683
   radius=6.7766
   shapeName=spine_03_capsule
   center=<-2.308 1.973 0.0583>
   rotation=<rot -0.0797 -90 -0.1694>
  }]
 thigh_r [{
   shapeType=Capsule
   length=26.4037
   radius=7.4199
   shapeName=thigh_r_capsule
   center=<20.2044 0.3836 -0.6065>
   rotation=<rot 0.0589 -95.9696 -94.3274>
  }]
}
constraints [
 {bone1Name=dyn_boba_drink_1, bone2Name=dyn_boba_drink_base, swing1LimitDegrees=20, swing2LimitDegrees=20, twistLimitDegrees=5}
 {bone1Name=dyn_boba_drink_2, bone2Name=dyn_boba_drink_1, swing1LimitDegrees=25, swing2LimitDegrees=25, twistLimitDegrees=5}
 {bone1Name=dyn_boba_drink_3, bone2Name=dyn_boba_drink_2, swing1LimitDegrees=30, swing2LimitDegrees=30, twistLimitDegrees=10}
 {bone1Name=dyn_bow, bone2Name=spine_03, swing1LimitDegrees=45, swing2LimitDegrees=45, twistLimitDegrees=10}
]
