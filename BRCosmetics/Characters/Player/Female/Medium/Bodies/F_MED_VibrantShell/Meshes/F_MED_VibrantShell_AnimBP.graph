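# Post-style AnimBP for the VibrantShell female cosmetic body, copying the attached parent's pose and adding secondary motion.
# A Control Rig feeds a pelvis-space RigidBody sim, then chained spring AnimDynamics for bracelets, anklet, butterfly wings and pendants; the bow node is unconnected.
# Each update derives RigidBody gravity from a parameter asset.
# The front-end check sets the rig's is_FE input and softens wing dynamics in menus.
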AnimBlueprint /BRCosmetics/Characters/Player/Female/Medium/Bodies/F_MED_VibrantShell/Meshes/F_MED_VibrantShell_AnimBP
target_skeleton=&/F_MED_VibrantShell_Skeleton
graphs [
 AnimGraph {nodes [
   Anim Dynamics - Bone: dyn_bow(Alpha=0.75, ExternalForce="0,0,0") {
    angularDampingOverride=0.99
    angularSpringConstant=250
    bAngularSpring=true
    bGravityOverrideInSimSpace=true
    bOverrideAngularDamping=true
    bOverrideLinearDamping=true
    bUseGravityOverride=true
    linearDampingOverride=0.99
    simulationSpace=BoneRelative
    alphaBoolBlend {blendOption=Linear}
    boundBone {boneName=dyn_bow}
    componentAppliedLinearAccClamp=<1000 1000 1000>
    gravityOverride {z=-980}
    physicsBodyDefinitions [{
      sphereCollisionRadius=0
      boundBone {boneName=dyn_bow}
      boxExtents=<20 15 15>
      constraintSetup {angularLimitsMax=<10 30 10>, angularLimitsMin=<-10 -30 -45>, angularTarget=<0 0 0>}
      localJointOffset=<8.5 0 0>
     }]
    relativeSpaceBone {boneName=root}
    simSpaceSettings {simSpaceAngularAlpha=1, externalAngularVelocity=<1 1 1>}
   }
   n1 = Get Alpha() -> (Alpha:float)
   n2 = Control Rig!(Source=Copy Pose From Mesh() {bUseAttachedParent=true}, Alpha=@n1, is_FE:double=Get Is FE() -> (Is FE:double)) -> (Pose) {controlRigAssetReference {blueprintRigClass=&/F_MED_VibrantShell_CtrlRig.F_MED_VibrantShell_CtrlRig_C}}
   n3 = RigidBody!(
    ComponentPose=Local To Component(LocalPose=@n2)
    Alpha=@n1
    OverrideWorldGravity:Vector=Get Override World Gravity() -> (Override World Gravity:Vector)
    ExternalForce
   ) -> (Pose) {
    bOverrideWorldGravity=true
    overridePhysicsAsset=&/F_MED_VibrantShell_Physics
    simulationSpace=BaseBoneSpace
    baseBoneRef {boneName=pelvis}
    componentAppliedLinearAccClamp=<1000 1000 1000>
    overrideWorldGravity {z=-980}
    simSpaceSettings {dampingAlpha=0, maxAngularAcceleration=4.5, maxAngularVelocity=2.5, maxLinearAcceleration=1200, maxLinearVelocity=450, worldAlpha=1}
   }
   n4 = Anim Dynamics - Bone: dyn_bracelet_1_l(ComponentPose=@n3, Alpha=@n1, ExternalForce="0,0,0") {
    angularDampingOverride=0.999
    angularSpringConstant=25
    bAngularSpring=true
    bGravityOverrideInSimSpace=true
    bOverrideAngularDamping=true
    bUseGravityOverride=true
    simulationSpace=RootRelative
    alphaBoolBlend {blendOption=Linear}
    boundBone {boneName=dyn_bracelet_1_l}
    componentAppliedLinearAccClamp=<3000 3000 3000>
    gravityOverride {z=-980}
    physicsBodyDefinitions [{
      sphereCollisionRadius=0
      boundBone {boneName=dyn_bracelet_1_l}
      boxExtents=<15 15 15>
      constraintSetup {angularLimitsMax=<10 2.5 10>, angularLimitsMin=<-10 -10 -10>, angularTarget=<0 0 0>}
      localJointOffset=<-0.1 0 0>
     }]
    simSpaceSettings {simSpaceAngularAlpha=1, externalAngularVelocity=<1 1 1>}
   }
   n5 = Anim Dynamics - Bone: dyn_bracelet_2_l(ComponentPose=@n4, Alpha=@n1, ExternalForce="0,0,0") {
    angularDampingOverride=0.999
    angularSpringConstant=25
    bAngularSpring=true
    bGravityOverrideInSimSpace=true
    bOverrideAngularDamping=true
    bUseGravityOverride=true
    simulationSpace=RootRelative
    alphaBoolBlend {blendOption=Linear}
    boundBone {boneName=dyn_bracelet_2_l}
    componentAppliedLinearAccClamp=<3000 3000 3000>
    gravityOverride {z=-980}
    physicsBodyDefinitions [{
      sphereCollisionRadius=0
      boundBone {boneName=dyn_bracelet_2_l}
      boxExtents=<15 15 15>
      constraintSetup {angularLimitsMax=<10 2.5 10>, angularLimitsMin=<-10 -10 -10>, angularTarget=<0 0 0>}
      localJointOffset=<-0.1 0 0>
     }]
    simSpaceSettings {simSpaceAngularAlpha=1, externalAngularVelocity=<1 1 1>}
   }
   n6 = Anim Dynamics - Bone: dyn_bracelet_r(ComponentPose=@n5, Alpha=@n1, ExternalForce="0,0,0") {
    angularDampingOverride=0.995
    angularSpringConstant=10
    bAngularSpring=true
    bGravityOverrideInSimSpace=true
    bOverrideAngularDamping=true
    bOverrideLinearDamping=true
    bUseGravityOverride=true
    linearDampingOverride=0.99
    simulationSpace=RootRelative
    alphaBoolBlend {blendOption=Linear}
    boundBone {boneName=dyn_bracelet_r}
    componentAppliedLinearAccClamp=<3000 3000 3000>
    gravityOverride {z=-980}
    physicsBodyDefinitions [{
      sphereCollisionRadius=0
      boundBone {boneName=dyn_bracelet_r}
      boxExtents=<10 10 10>
      constraintSetup {angularLimitsMax=<15 2.5 2.5>, angularLimitsMin=<-15 -2.5 -2.5>, angularTarget=<0 0 0>}
      localJointOffset=<1 0 0>
     }]
    simSpaceSettings {simSpaceAngularAlpha=1, externalAngularVelocity=<1 1 1>}
   }
   n7 = Anim Dynamics - Bone: dyn_ankle_bracelet_r(ComponentPose=@n6, Alpha=@n1, ExternalForce="0,0,0") {
    angularDampingOverride=0.995
    angularSpringConstant=10
    bAngularSpring=true
    bGravityOverrideInSimSpace=true
    bOverrideAngularDamping=true
    bUseGravityOverride=true
    simulationSpace=RootRelative
    alphaBoolBlend {blendOption=Linear}
    boundBone {boneName=dyn_ankle_bracelet_r}
    componentAppliedLinearAccClamp=<3000 3000 3000>
    gravityOverride {z=-980}
    physicsBodyDefinitions [{
      sphereCollisionRadius=0
      boundBone {boneName=dyn_ankle_bracelet_r}
      boxExtents=<30 30 30>
      constraintSetup {angularLimitsMax=<10 10 5>, angularLimitsMin=<-10 -10 -10>, angularTarget=<0 0 0>}
      localJointOffset=<0 -1 0>
     }]
    simSpaceSettings {simSpaceAngularAlpha=1, externalAngularVelocity=<1 1 1>}
   }
   n8 = Get Alpha_Butterfly_wings() -> (Alpha_Butterfly_wings:float)
   n9 = Anim Dynamics - Bone: dyn_butterfly_wing_1_l(ComponentPose=@n7, Alpha=@n8, ExternalForce="0,0,0") {
    angularDampingOverride=0.995
    angularSpringConstant=30
    bAngularSpring=true
    bGravityOverrideInSimSpace=true
    bOverrideAngularDamping=true
    bOverrideLinearDamping=true
    bUseGravityOverride=true
    linearDampingOverride=0.9
    simulationSpace=RootRelative
    alphaBoolBlend {blendOption=Linear}
    alphaScaleBiasClamp {bInterpResult=true}
    boundBone {boneName=dyn_butterfly_wing_1_l}
    componentAppliedLinearAccClamp=<3000 3000 3000>
    gravityOverride {z=-980}
    physicsBodyDefinitions [{
      sphereCollisionRadius=0
      boundBone {boneName=dyn_butterfly_wing_1_l}
      boxExtents=<25 25 25>
      constraintSetup {angularLimitsMax=<10 10 5>, angularLimitsMin=<-10 -10 -10>, angularTarget=<0 0 0>}
      localJointOffset=<0.005 0 -0.25>
     }]
    simSpaceSettings {simSpaceAngularAlpha=1, externalAngularVelocity=<1 1 1>}
   }
   n10 = Anim Dynamics - Bone: dyn_butterfly_wing_2_l(ComponentPose=@n9, Alpha=@n8, ExternalForce="0,0,0") {
    angularDampingOverride=0.995
    angularSpringConstant=30
    bAngularSpring=true
    bGravityOverrideInSimSpace=true
    bOverrideAngularDamping=true
    bOverrideLinearDamping=true
    bUseGravityOverride=true
    linearDampingOverride=0.9
    simulationSpace=RootRelative
    alphaBoolBlend {blendOption=Linear}
    alphaScaleBiasClamp {bInterpResult=true}
    boundBone {boneName=dyn_butterfly_wing_2_l}
    componentAppliedLinearAccClamp=<3000 3000 3000>
    gravityOverride {z=-980}
    physicsBodyDefinitions [{
      sphereCollisionRadius=0
      boundBone {boneName=dyn_butterfly_wing_2_l}
      boxExtents=<25 25 25>
      constraintSetup {angularLimitsMax=<10 10 5>, angularLimitsMin=<-10 -10 -10>, angularTarget=<0 0 0>}
      localJointOffset=<0.005 0 0.25>
     }]
    simSpaceSettings {simSpaceAngularAlpha=1, externalAngularVelocity=<1 1 1>}
   }
   n11 = Anim Dynamics - Bone: dyn_butterfly_wing_1_r(ComponentPose=@n10, Alpha=@n8, ExternalForce="0,0,0") {
    angularDampingOverride=0.995
    angularSpringConstant=30
    bAngularSpring=true
    bGravityOverrideInSimSpace=true
    bOverrideAngularDamping=true
    bOverrideLinearDamping=true
    bUseGravityOverride=true
    linearDampingOverride=0.9
    simulationSpace=RootRelative
    alphaBoolBlend {blendOption=Linear}
    alphaScaleBiasClamp {bInterpResult=true}
    boundBone {boneName=dyn_butterfly_wing_1_r}
    componentAppliedLinearAccClamp=<3000 3000 3000>
    gravityOverride {z=-980}
    physicsBodyDefinitions [{
      sphereCollisionRadius=0
      boundBone {boneName=dyn_butterfly_wing_1_r}
      boxExtents=<25 25 25>
      constraintSetup {angularLimitsMax=<10 10 5>, angularLimitsMin=<-10 -10 -10>, angularTarget=<0 0 0>}
      localJointOffset=<0.005 0 -0.25>
     }]
    simSpaceSettings {simSpaceAngularAlpha=1, externalAngularVelocity=<1 1 1>}
   }
   n12 = Anim Dynamics - Bone: dyn_butterfly_wing_2_r(ComponentPose=@n11, Alpha=@n8, ExternalForce="0,0,0") {
    angularDampingOverride=0.995
    angularSpringConstant=30
    bAngularSpring=true
    bGravityOverrideInSimSpace=true
    bOverrideAngularDamping=true
    bOverrideLinearDamping=true
    bUseGravityOverride=true
    linearDampingOverride=0.9
    simulationSpace=RootRelative
    alphaBoolBlend {blendOption=Linear}
    alphaScaleBiasClamp {bInterpResult=true}
    boundBone {boneName=dyn_butterfly_wing_2_r}
    componentAppliedLinearAccClamp=<3000 3000 3000>
    gravityOverride {z=-980}
    physicsBodyDefinitions [{
      sphereCollisionRadius=0
      boundBone {boneName=dyn_butterfly_wing_2_r}
      boxExtents=<25 25 25>
      constraintSetup {angularLimitsMax=<10 10 5>, angularLimitsMin=<-10 -10 -10>, angularTarget=<0 0 0>}
      localJointOffset=<0.005 0 0.25>
     }]
    simSpaceSettings {simSpaceAngularAlpha=1, externalAngularVelocity=<1 1 1>}
   }
   n13 = Get Alpha_0() -> (Alpha_0:float)
   n14 = Anim Dynamics - Bone: dyn_pendant_1_l(ComponentPose=@n12, Alpha=@n13, ExternalForce="0,0,0") {
    angularDampingOverride=0.9995
    angularSpringConstant=7.5
    bAngularSpring=true
    bGravityOverrideInSimSpace=true
    bOverrideAngularDamping=true
    bOverrideLinearDamping=true
    bUseGravityOverride=true
    linearDampingOverride=0.999
    simulationSpace=RootRelative
    alphaBoolBlend {blendOption=Linear}
    boundBone {boneName=dyn_pendant_1_l}
    componentAppliedLinearAccClamp=<3000 3000 3000>
    gravityOverride {z=-980}
    physicsBodyDefinitions [{
      sphereCollisionRadius=0
      boundBone {boneName=dyn_pendant_1_l}
      boxExtents=<7.5 4 4>
      constraintSetup {angularLimitsMax=<30 2.5 30>, angularLimitsMin=<-30 -30 -30>, angularTarget=<0 0 0>}
      localJointOffset=<1 0 0>
     }]
    simSpaceSettings {simSpaceAngularAlpha=1, externalAngularVelocity=<1 1 1>}
   }
   n15 = Anim Dynamics - Bone: dyn_pendant_2_l(ComponentPose=@n14, Alpha=@n13, ExternalForce="0,0,0") {
    angularDampingOverride=0.9995
    angularSpringConstant=7.5
    bAngularSpring=true
    bGravityOverrideInSimSpace=true
    bOverrideAngularDamping=true
    bOverrideLinearDamping=true
    bUseGravityOverride=true
    linearDampingOverride=0.999
    simulationSpace=RootRelative
    alphaBoolBlend {blendOption=Linear}
    boundBone {boneName=dyn_pendant_2_l}
    componentAppliedLinearAccClamp=<3000 3000 3000>
    gravityOverride {z=-980}
    physicsBodyDefinitions [{
      sphereCollisionRadius=0
      boundBone {boneName=dyn_pendant_2_l}
      boxExtents=<7.5 4 4>
      constraintSetup {angularLimitsMax=<30 2.5 30>, angularLimitsMin=<-30 -30 -30>, angularTarget=<0 0 0>}
      localJointOffset=<1 0 0>
     }]
    simSpaceSettings {simSpaceAngularAlpha=1, externalAngularVelocity=<1 1 1>}
   }
   n16 = Anim Dynamics - Bone: dyn_pendant_3_l(ComponentPose=@n15, Alpha=@n13, ExternalForce="0,0,0") {
    angularDampingOverride=0.9995
    angularSpringConstant=7.5
    bAngularSpring=true
    bGravityOverrideInSimSpace=true
    bOverrideAngularDamping=true
    bOverrideLinearDamping=true
    bUseGravityOverride=true
    linearDampingOverride=0.999
    simulationSpace=RootRelative
    alphaBoolBlend {blendOption=Linear}
    boundBone {boneName=dyn_pendant_3_l}
    componentAppliedLinearAccClamp=<3000 3000 3000>
    gravityOverride {z=-980}
    physicsBodyDefinitions [{
      sphereCollisionRadius=0
      boundBone {boneName=dyn_pendant_3_l}
      boxExtents=<7.5 4 4>
      constraintSetup {angularLimitsMax=<30 2.5 10>, angularLimitsMin=<-30 -20 -10>, angularTarget=<0 0 0>}
      localJointOffset=<1 0 0>
     }]
    simSpaceSettings {simSpaceAngularAlpha=1, externalAngularVelocity=<1 1 1>}
   }
   n17 = Anim Dynamics - Bone: dyn_pendant_1_r(ComponentPose=@n16, Alpha=@n13, ExternalForce="0,0,0") {
    angularDampingOverride=0.9995
    angularSpringConstant=7.5
    bAngularSpring=true
    bGravityOverrideInSimSpace=true
    bOverrideAngularDamping=true
    bOverrideLinearDamping=true
    bUseGravityOverride=true
    linearDampingOverride=0.999
    simulationSpace=RootRelative
    alphaBoolBlend {blendOption=Linear}
    boundBone {boneName=dyn_pendant_1_r}
    componentAppliedLinearAccClamp=<3000 3000 3000>
    gravityOverride {z=-980}
    physicsBodyDefinitions [{
      sphereCollisionRadius=0
      boundBone {boneName=dyn_pendant_1_r}
      boxExtents=<7.5 4 4>
      constraintSetup {angularLimitsMax=<30 30 30>, angularLimitsMin=<-30 -2.5 -30>, angularTarget=<0 0 0>}
      localJointOffset=<1 0 0>
     }]
    simSpaceSettings {simSpaceAngularAlpha=1, externalAngularVelocity=<1 1 1>}
   }
   n18 = Anim Dynamics - Bone: dyn_pendant_2_r(ComponentPose=@n17, Alpha=@n13, ExternalForce="0,0,0") {
    angularDampingOverride=0.9995
    angularSpringConstant=7.5
    bAngularSpring=true
    bGravityOverrideInSimSpace=true
    bOverrideAngularDamping=true
    bOverrideLinearDamping=true
    bUseGravityOverride=true
    linearDampingOverride=0.999
    simulationSpace=RootRelative
    alphaBoolBlend {blendOption=Linear}
    boundBone {boneName=dyn_pendant_2_r}
    componentAppliedLinearAccClamp=<3000 3000 3000>
    gravityOverride {z=-980}
    physicsBodyDefinitions [{
      sphereCollisionRadius=0
      boundBone {boneName=dyn_pendant_2_r}
      boxExtents=<7.5 4 4>
      constraintSetup {angularLimitsMax=<30 30 30>, angularLimitsMin=<-30 -0.5 -30>, angularTarget=<0 0 0>}
      localJointOffset=<1 0 0>
     }]
    simSpaceSettings {simSpaceAngularAlpha=1, externalAngularVelocity=<1 1 1>}
   }
   n19 = Anim Dynamics - Bone: dyn_pendant_3_r(ComponentPose=@n18, Alpha=@n13, ExternalForce="0,0,0") {
    angularDampingOverride=0.9995
    angularSpringConstant=7.5
    bAngularSpring=true
    bGravityOverrideInSimSpace=true
    bOverrideAngularDamping=true
    bOverrideLinearDamping=true
    bUseGravityOverride=true
    linearDampingOverride=0.999
    simulationSpace=RootRelative
    alphaBoolBlend {blendOption=Linear}
    boundBone {boneName=dyn_pendant_3_r}
    componentAppliedLinearAccClamp=<3000 3000 3000>
    gravityOverride {z=-980}
    physicsBodyDefinitions [{
      sphereCollisionRadius=0
      boundBone {boneName=dyn_pendant_3_r}
      boxExtents=<7.5 4 4>
      constraintSetup {angularLimitsMax=<20 20 20>, angularLimitsMin=<-20 -2.5 -20>, angularTarget=<0 0 0>}
      localJointOffset=<1 0 0>
     }]
    simSpaceSettings {simSpaceAngularAlpha=1, externalAngularVelocity=<1 1 1>}
   }
   Output Pose(Result=Component To Local(ComponentPose=@n19)) {layerGroup=None}
  ]}
 EventGraph {nodes [
   Event BlueprintUpdateAnimation()
   > n1 = Sequence()
   > n2 = "Gravity Override State Function (Parameter Asset) (Nativized)"(
    CharacterPartAnimInstance=Self-Reference()
    GameTimeInSeconds=GetGameTimeInSeconds()
    Parameters=&/F_MED_VibrantShell_Parameters
   )
   > Set Override World Gravity(Override World Gravity:Vector=@n2) -> (Output_Get:Vector)
   n3 = Branch(execute=@n1.then_1, Condition=Get bIsInFrontEnd())
   > Set Alpha_Butterfly_wings(Alpha_Butterfly_wings:float=0.66) -> (Output_Get:float)
   > Set Is FE(Is FE:double=1) -> (Output_Get:double)
   Set Alpha_Butterfly_wings(execute=@n3.else, Alpha_Butterfly_wings=1)
   > Set Is FE(Is FE=0)
  ]}
]
variables [
 Alpha {type=number, default=1}
 Alpha_0 {type=number, default=0.75}
 Alpha_Butterfly_wings {type=number, default=1}
 Is FE {type=number, default=1}
 Override World Gravity {type=Vector, default=<0 0 -980>}
]
