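# Physics asset of cape colliders for the F_MED_VibrantShell female medium player cosmetic body, presumably used by cape simulation.
# Capsules cover the pelvis, spine, neck and clavicles, with extra capsules on spine_02 and spine_05 and a sphere on the head.
# The dyn_boba_drink_3 bone also carries a capsule, while arm, leg and foot bodies have no shapes.
# Constraints link the body bones with all angular motion free; the drink body is unconstrained.
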PhysicsAsset /BRCosmetics/Characters/Player/Female/Medium/Bodies/F_MED_VibrantShell/Meshes/F_MED_VibrantShell_CapeColliders
bodies {
 calf_l []
 calf_r []
 clavicle_l [{shapeType=Capsule, length=14.1962, radius=4.7101, center=<9.4266 -1.1806 -2.4977>, rotation=<rot 69.3297 1.4476 176.9462>}]
 clavicle_r [{shapeType=Capsule, length=14.1962, radius=4.7101, center=<-9.4266 1.1806 2.4977>, rotation=<rot 69.3297 1.4476 -3.0538>}]
 dyn_boba_drink_3 [{
   shapeType=Capsule
   length=3.66
   radius=9.8678
   shapeName=dyn_boba_drink_3_capsule
   center=<5.1935 0.1513 -0.1502>
   rotation=<rot 0 90.0711 -89.4766>
  }]
 foot_l []
 foot_r []
 head [{shapeType=Sphere, radius=17.3125, center=<11.028 -0.174 0>}]
 lowerarm_l []
 lowerarm_r []
 neck_02 [{shapeType=Capsule, length=13.4375, radius=7.3175, center=<-0.2642 -0.3408 0>, rotation=<rot 90 -171.1193 -175.3978>}]
 pelvis [{shapeType=Capsule, length=22.1697, radius=13.2487, center=<-7.3962 0.2381 3.305>, rotation=<rot 0 -0.0012 5>}]
 spine_01 [{shapeType=Capsule, length=10.7463, radius=10.465, center=<-1.7965 0.1884 0>}]
 spine_02 [
  {shapeType=Capsule, length=15.5474, radius=5.1996, center=<-1.711 -1.5542 0.0001>}
  {shapeType=Capsule, length=35.1456, radius=5.8307, center=<2.5024 0.4992 3.9825>, rotation=<rot 90 0.0049 0.0049>}
  {shapeType=Capsule, length=35.625, radius=5.3513, center=<2.4704 0.4991 -3.9825>, rotation=<rot 90 -0.0396 -0.0396>}
 ]
 spine_03 [{shapeType=Capsule, length=6.5799, radius=7.0091, center=<0.1445 0.4289 0.0001>}]
 spine_04 [{shapeType=Capsule, length=16.9727, radius=4.2597, center=<2.766 -2.9253 0.0002>}]
 spine_05 [
  {shapeType=Capsule, length=20.2454, radius=3.8823, center=<-2.2795 1.8109 0>}
  {shapeType=Capsule, length=16.9518, radius=10.9425, center=<-0.2916 0.4452 0>}
 ]
 thigh_l []
 thigh_r []
 upperarm_l []
 upperarm_r []
}
constraints [
 {
  bone1Name=spine_01
  bone2Name=pelvis
  swing1LimitDegrees=45
  swing1Motion=Free
  swing2LimitDegrees=45
  swing2Motion=Free
  twistLimitDegrees=45
  twistMotion=Free
 }
 {
  bone1Name=spine_02
  bone2Name=spine_01
  swing1LimitDegrees=45
  swing1Motion=Free
  swing2LimitDegrees=45
  swing2Motion=Free
  twistLimitDegrees=45
  twistMotion=Free
 }
 {
  bone1Name=spine_03
  bone2Name=spine_02
  swing1LimitDegrees=45
  swing1Motion=Free
  swing2LimitDegrees=45
  swing2Motion=Free
  twistLimitDegrees=45
  twistMotion=Free
 }
 {
  bone1Name=spine_04
  bone2Name=spine_03
  swing1LimitDegrees=45
  swing1Motion=Free
  swing2LimitDegrees=45
  swing2Motion=Free
  twistLimitDegrees=45
  twistMotion=Free
 }
 {
  bone1Name=spine_05
  bone2Name=spine_04
  swing1LimitDegrees=45
  swing1Motion=Free
  swing2LimitDegrees=45
  swing2Motion=Free
  twistLimitDegrees=45
  twistMotion=Free
 }
 {
  bone1Name=neck_02
  bone2Name=spine_05
  swing1LimitDegrees=45
  swing1Motion=Free
  swing2LimitDegrees=45
  swing2Motion=Free
  twistLimitDegrees=45
  twistMotion=Free
 }
 {
  bone1Name=head
  bone2Name=neck_02
  swing1LimitDegrees=45
  swing1Motion=Free
  swing2LimitDegrees=45
  swing2Motion=Free
  twistLimitDegrees=45
  twistMotion=Free
 }
 {
  bone1Name=thigh_l
  bone2Name=pelvis
  swing1LimitDegrees=45
  swing1Motion=Free
  swing2LimitDegrees=45
  swing2Motion=Free
  twistLimitDegrees=45
  twistMotion=Free
 }
 {
  bone1Name=foot_l
  bone2Name=calf_l
  swing1LimitDegrees=45
  swing1Motion=Free
  swing2LimitDegrees=45
  swing2Motion=Free
  twistLimitDegrees=45
  twistMotion=Free
 }
 {
  bone1Name=clavicle_r
  bone2Name=spine_05
  swing1LimitDegrees=45
  swing1Motion=Free
  swing2LimitDegrees=45
  swing2Motion=Free
  twistLimitDegrees=45
  twistMotion=Free
 }
 {
  bone1Name=upperarm_r
  bone2Name=clavicle_r
  swing1LimitDegrees=45
  swing1Motion=Free
  swing2LimitDegrees=45
  swing2Motion=Free
  twistLimitDegrees=45
  twistMotion=Free
 }
 {
  bone1Name=lowerarm_r
  bone2Name=upperarm_r
  swing1LimitDegrees=45
  swing1Motion=Free
  swing2LimitDegrees=45
  swing2Motion=Free
  twistLimitDegrees=45
  twistMotion=Free
 }
 {
  bone1Name=foot_r
  bone2Name=calf_r
  swing1LimitDegrees=45
  swing1Motion=Free
  swing2LimitDegrees=45
  swing2Motion=Free
  twistLimitDegrees=45
  twistMotion=Free
 }
 {
  bone1Name=lowerarm_l
  bone2Name=upperarm_l
  swing1LimitDegrees=45
  swing1Motion=Free
  swing2LimitDegrees=45
  swing2Motion=Free
  twistLimitDegrees=45
  twistMotion=Free
 }
 {
  bone1Name=upperarm_l
  bone2Name=clavicle_l
  swing1LimitDegrees=45
  swing1Motion=Free
  swing2LimitDegrees=45
  swing2Motion=Free
  twistLimitDegrees=45
  twistMotion=Free
 }
 {
  bone1Name=clavicle_l
  bone2Name=spine_05
  swing1LimitDegrees=45
  swing1Motion=Free
  swing2LimitDegrees=45
  swing2Motion=Free
  twistLimitDegrees=45
  twistMotion=Free
 }
 {
  bone1Name=calf_r
  bone2Name=thigh_r
  swing1LimitDegrees=45
  swing1Motion=Free
  swing2LimitDegrees=45
  swing2Motion=Free
  twistLimitDegrees=45
  twistMotion=Free
 }
 {
  bone1Name=thigh_r
  bone2Name=pelvis
  swing1LimitDegrees=45
  swing1Motion=Free
  swing2LimitDegrees=45
  swing2Motion=Free
  twistLimitDegrees=45
  twistMotion=Free
 }
 {
  bone1Name=calf_l
  bone2Name=thigh_l
  swing1LimitDegrees=45
  swing1Motion=Free
  swing2LimitDegrees=45
  swing2Motion=Free
  twistLimitDegrees=45
  twistMotion=Free
 }
]
