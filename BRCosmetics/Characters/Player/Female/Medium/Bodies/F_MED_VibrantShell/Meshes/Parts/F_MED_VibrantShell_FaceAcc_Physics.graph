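# Physics asset for the VibrantShell medium female character's FaceAcc mesh part, covering secondary motion of hair, earrings and necklace.
# Capsule bodies on head, neck, spine, clavicles and upper arms accompany dyn_ chains for a hair tail, side locks, earrings and necklace.
# Chain links use swing and twist limits, while hair-tail and earring bases rotate freely against head and neck.
# The necklace base is rotation-locked to spine_05; hair-lock bases have no constraint to the head.
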PhysicsAsset /BRCosmetics/Characters/Player/Female/Medium/Bodies/F_MED_VibrantShell/Meshes/Parts/F_MED_VibrantShell_FaceAcc_Physics
bodies {
 clavicle_l [{
   shapeType=Capsule
   length=9.505
   radius=4.505
   shapeName=clavicle_l_capsule
   center=<5.1567 -0.7782 0.1723>
   rotation=<rot -0.0002 -90.0001 100>
  }]
 clavicle_r [{
   shapeType=Capsule
   length=9.505
   radius=4.505
   shapeName=clavicle_r_capsule
   center=<-4.8491 0.7782 -0.1723>
   rotation=<rot -0.0002 -90.0001 100>
  }]
 dyn_earring_base_l [{
   shapeType=Capsule
   length=0
   radius=0.9013
   shapeName=dyn_earring_base_l_capsule
   center=<0.0026 -0.0005 -0.0088>
   rotation=<rot 0 89.9598 -89.6511>
  }]
 dyn_earring_base_r [{
   shapeType=Capsule
   length=0
   radius=0.9066
   shapeName=dyn_earring_base_r_capsule
   center=<-0.0005 0.0067 0.0071>
   rotation=<rot 0 89.9076 -89.9444>
  }]
 dyn_earring_l [{
   shapeType=Capsule
   length=3.3094
   radius=1.4013
   shapeName=dyn_earring_l_capsule
   center=<3.9046 -0.0032 -0.0062>
   rotation=<rot 0.0606 79.96 -89.6564>
  }]
 dyn_earring_r [{
   shapeType=Capsule
   length=3.3117
   radius=1.4066
   shapeName=dyn_earring_r_capsule
   center=<3.9015 0.0004 0.0109>
   rotation=<rot 0.0096 79.9076 -89.9453>
  }]
 dyn_hair_tail_1 [{
   shapeType=Capsule
   length=7.9576
   radius=5.433
   shapeName=dyn_hair_tail_1_capsule
   center=<3.2781 -0.0624 -1.4222>
   rotation=<rot 0 89.4143 -77.952>
  }]
 dyn_hair_tail_2 [{
   shapeType=Capsule
   length=11.4527
   radius=4.0413
   shapeName=dyn_hair_tail_2_capsule
   center=<6.3844 0.2107 0.4867>
   rotation=<rot 9.973 -89.4895 -95.7459>
  }]
 dyn_hair_tail_3 [{
   shapeType=Capsule
   length=6.9888
   radius=4.5004
   shapeName=dyn_hair_tail_3_capsule
   center=<6.2633 -0.56 -0.6466>
   rotation=<rot -9.2768 -172.7548 -66.0655>
  }]
 dyn_hair_tail_4 [{
   shapeType=Capsule
   length=7.2569
   radius=4.36
   shapeName=dyn_hair_tail_4_capsule
   center=<6.8646 1.1304 -1.7541>
   rotation=<rot 41.0165 104.5002 -67.9786>
  }]
 dyn_hair_tail_base [{shapeType=Capsule, length=0, radius=3, shapeName=dyn_hair_tail_base_capsule, rotation=<rot 0 -90 0>}]
 dyn_lock_of_hair_1_1_r [{
   shapeType=Capsule
   length=3.1898
   radius=1
   shapeName=dyn_lock_of_hair_1_1_r_capsule
   center=<4.1111 -0.3714 0.4909>
   rotation=<rot 0 78.8886 -77.0857>
  }]
 dyn_lock_of_hair_1_base_l [{shapeType=Capsule, length=0.01, radius=1.5, shapeName=dyn_lock_of_hair_1_base_l_capsule, rotation=<rot 0 -90 0>}]
 dyn_lock_of_hair_1_base_r [{shapeType=Capsule, length=0.01, radius=1.5, shapeName=dyn_lock_of_hair_1_base_r_capsule, rotation=<rot 0 -90 0>}]
 dyn_lock_of_hair_1_l [{
   shapeType=Capsule
   length=3.9789
   radius=0.9
   shapeName=dyn_lock_of_hair_1_l_capsule
   center=<3.5423 -1.1199 -0.7394>
   rotation=<rot -0.0001 -122.4971 -66.875>
  }]
 dyn_lock_of_hair_2_1_l [{
   shapeType=Capsule
   length=4.8571
   radius=0.6768
   shapeName=dyn_lock_of_hair_2_1_l_capsule
   center=<3.3035 0.0051 -0.0035>
   rotation=<rot 0 91.2969 -88.618>
  }]
 dyn_lock_of_hair_2_1_r [{
   shapeType=Capsule
   length=7.1382
   radius=2.3518
   shapeName=dyn_lock_of_hair_2_1_r_capsule
   center=<4.7216 0.2045 -0.6001>
   rotation=<rot -0.4492 92.8753 -92.4523>
  }]
 dyn_lock_of_hair_2_2_l [{
   shapeType=Capsule
   length=5.0326
   radius=0.9436
   shapeName=dyn_lock_of_hair_2_2_l_capsule
   center=<3.32 0.1959 -0.0068>
   rotation=<rot -0.0745 93.4395 -89.1482>
  }]
 dyn_lock_of_hair_2_2_r [{
   shapeType=Capsule
   length=2.589
   radius=1.6227
   shapeName=dyn_lock_of_hair_2_2_r_capsule
   center=<4.7225 0.0267 -0.1872>
   rotation=<rot 10.7959 -7.1556 -63.3486>
  }]
 dyn_lock_of_hair_2_3_l [{
   shapeType=Capsule
   length=3.654
   radius=0.8266
   shapeName=dyn_lock_of_hair_2_3_l_capsule
   center=<2.5749 0.2199 0.017>
   rotation=<rot 0 89.7484 -89.8798>
  }]
 dyn_lock_of_hair_2_3_r [{
   shapeType=Capsule
   length=10.0885
   radius=1.4934
   shapeName=dyn_lock_of_hair_2_3_r_capsule
   center=<4.9311 -0.5437 -0.2283>
   rotation=<rot -14.5378 -93.314 -83.2646>
  }]
 dyn_lock_of_hair_2_base_l [{shapeType=Capsule, length=0.01, radius=1.5, shapeName=dyn_lock_of_hair_2_base_l_capsule, rotation=<rot 0 -90 0>}]
 dyn_lock_of_hair_2_base_r [{shapeType=Capsule, length=0.01, radius=1.5, shapeName=dyn_lock_of_hair_2_base_r_capsule, rotation=<rot 0 -90 0>}]
 dyn_lock_of_hair_3_1_l [{
   shapeType=Capsule
   length=2.9603
   radius=1.1886
   shapeName=dyn_lock_of_hair_3_1_l_capsule
   center=<2.8164 0.0056 -0.0108>
   rotation=<rot 0 -89.3433 -88.5336>
  }]
 dyn_lock_of_hair_3_1_r [{
   shapeType=Capsule
   length=3.7339
   radius=1.0305
   shapeName=dyn_lock_of_hair_3_1_r_capsule
   center=<2.7337 -0.0142 -0.0675>
   rotation=<rot 0 -91.1775 -87.9821>
  }]
 dyn_lock_of_hair_3_2_l [{
   shapeType=Capsule
   length=5.6948
   radius=1.1379
   shapeName=dyn_lock_of_hair_3_2_l_capsule
   center=<3.8746 0.0149 0.0116>
   rotation=<rot 0 92.0966 -88.4168>
  }]
 dyn_lock_of_hair_3_2_r [{
   shapeType=Capsule
   length=5.6146
   radius=1.228
   shapeName=dyn_lock_of_hair_3_2_r_capsule
   center=<5.9633 -0.234 0.008>
   rotation=<rot 1.6341 -74.4257 -99.4723>
  }]
 dyn_lock_of_hair_3_3_l [{
   shapeType=Capsule
   length=5.917
   radius=0.9207
   shapeName=dyn_lock_of_hair_3_3_l_capsule
   center=<4.4509 -0.9248 0.1728>
   rotation=<rot 0.7803 -85.6044 -85.5707>
  }]
 dyn_lock_of_hair_3_base_l [{shapeType=Capsule, length=0.01, radius=1.5, shapeName=dyn_lock_of_hair_3_base_l_capsule, rotation=<rot 0 -90 0>}]
 dyn_lock_of_hair_3_base_r [{shapeType=Capsule, length=0.01, radius=1.5, shapeName=dyn_lock_of_hair_3_base_r_capsule, rotation=<rot 0 -90 0>}]
 dyn_necklace_1 [
  {
   shapeType=Capsule
   length=5.7227
   radius=1.1761
   shapeName=dyn_necklace_1_capsule
   center=<9.1076 0.767 -2.5043>
   rotation=<rot -2.4299 -109.9715 51.0145>
  }
  {
   shapeType=Capsule
   length=5.5
   radius=0.75
   shapeName=dyn_necklace_1_capsule
   center=<9.065 0.3576 2.2522>
   rotation=<rot 0.5383 -104.9786 -54.494>
  }
 ]
 dyn_necklace_base [{
   shapeType=Capsule
   length=1
   radius=5.4261
   shapeName=dyn_necklace_base_capsule
   center=<0.0577 -0.0099 0.0284>
   rotation=<rot -4.6163 -80.0075 94.4163>
  }]
 head [{
   shapeType=Capsule
   length=10.12
   radius=6.7574
   shapeName=head_capsule
   center=<5.5019 1.5477 0.2464>
   rotation=<rot 1.6484 61.0727 -89.2311>
  }]
 neck_01 [{shapeType=Capsule, length=3.255, radius=3.755, shapeName=neck_01_capsule, center=<3.3234 0 0>, rotation=<rot 0 -90.0003 90>}]
 spine_02 [{shapeType=Capsule, length=10.25, radius=6.5, shapeName=spine_02_capsule, rotation=<rot 0 -90 0>}]
 spine_04 [{shapeType=Capsule, length=10.125, radius=6.875, shapeName=spine_04_capsule, center=<1.2753 -0.1488 0>, rotation=<rot 0 -90 0>}]
 spine_05 [{shapeType=Capsule, length=10.625, radius=6.875, shapeName=spine_05_capsule, center=<3.6693 -0.6267 0>, rotation=<rot 0 -90 0>}]
 upperarm_l [{
   shapeType=Capsule
   length=23.38
   radius=3.88
   shapeName=upperarm_l_capsule
   center=<13.9714 -1.1992 -0.018>
   rotation=<rot 3.7295 -93.3317 89.9511>
  }]
 upperarm_r [{
   shapeType=Capsule
   length=23.38
   radius=3.88
   shapeName=upperarm_r_capsule
   center=<-14.5863 1.1992 0.018>
   rotation=<rot 3.7295 -93.3317 89.9511>
  }]
}
constraints [
 {bone1Name=dyn_earring_l, bone2Name=dyn_earring_base_l, swing1LimitDegrees=30, swing2LimitDegrees=30, twistLimitDegrees=30}
 {bone1Name=dyn_earring_r, bone2Name=dyn_earring_base_r, swing1LimitDegrees=30, swing2LimitDegrees=30, twistLimitDegrees=30}
 {
  bone1Name=dyn_hair_tail_base
  bone2Name=head
  swing1LimitDegrees=45
  swing1Motion=Free
  swing2LimitDegrees=45
  swing2Motion=Free
  twistLimitDegrees=45
  twistMotion=Free
 }
 {bone1Name=dyn_hair_tail_1, bone2Name=dyn_hair_tail_base, swing1LimitDegrees=20, swing2LimitDegrees=45, twistLimitDegrees=5}
 {bone1Name=dyn_hair_tail_2, bone2Name=dyn_hair_tail_1, swing1LimitDegrees=20, swing2LimitDegrees=20, twistLimitDegrees=10}
 {bone1Name=dyn_hair_tail_3, bone2Name=dyn_hair_tail_2, swing1LimitDegrees=25, swing2LimitDegrees=25, twistLimitDegrees=10}
 {bone1Name=dyn_hair_tail_4, bone2Name=dyn_hair_tail_3, swing1LimitDegrees=30, swing2LimitDegrees=30, twistLimitDegrees=10}
 {
  bone1Name=dyn_earring_base_l
  bone2Name=neck_01
  swing1LimitDegrees=45
  swing1Motion=Free
  swing2LimitDegrees=45
  swing2Motion=Free
  twistLimitDegrees=45
  twistMotion=Free
 }
 {
  bone1Name=dyn_earring_base_r
  bone2Name=neck_01
  swing1LimitDegrees=45
  swing1Motion=Free
  swing2LimitDegrees=45
  swing2Motion=Free
  twistLimitDegrees=45
  twistMotion=Free
 }
 {
  bone1Name=upperarm_r
  bone2Name=clavicle_r
  swing1LimitDegrees=45
  swing1Motion=Free
  swing2LimitDegrees=45
  swing2Motion=Free
  twistLimitDegrees=45
  twistMotion=Free
 }
 {
  bone1Name=upperarm_l
  bone2Name=clavicle_l
  swing1LimitDegrees=45
  swing1Motion=Free
  swing2LimitDegrees=45
  swing2Motion=Free
  twistLimitDegrees=45
  twistMotion=Free
 }
 {bone1Name=dyn_lock_of_hair_2_1_l, bone2Name=dyn_lock_of_hair_2_base_l, swing1LimitDegrees=10, swing2LimitDegrees=10, twistLimitDegrees=5}
 {bone1Name=dyn_lock_of_hair_2_2_l, bone2Name=dyn_lock_of_hair_2_1_l, swing1LimitDegrees=15, swing2LimitDegrees=15, twistLimitDegrees=5}
 {bone1Name=dyn_lock_of_hair_2_3_l, bone2Name=dyn_lock_of_hair_2_2_l, swing1LimitDegrees=20, swing2LimitDegrees=20, twistLimitDegrees=10}
 {bone1Name=dyn_lock_of_hair_2_1_r, bone2Name=dyn_lock_of_hair_2_base_r, swing1LimitDegrees=5, swing2LimitDegrees=5, twistLimitDegrees=5}
 {bone1Name=dyn_lock_of_hair_2_2_r, bone2Name=dyn_lock_of_hair_2_1_r, swing1LimitDegrees=20, swing2LimitDegrees=20, twistLimitDegrees=20}
 {bone1Name=dyn_lock_of_hair_2_3_r, bone2Name=dyn_lock_of_hair_2_2_r, swing1LimitDegrees=20, swing2LimitDegrees=20, twistLimitDegrees=10}
 {bone1Name=dyn_lock_of_hair_3_1_l, bone2Name=dyn_lock_of_hair_3_base_l, swing1LimitDegrees=20, swing2LimitDegrees=20, twistLimitDegrees=31}
 {bone1Name=dyn_lock_of_hair_3_2_l, bone2Name=dyn_lock_of_hair_3_1_l, swing1LimitDegrees=30, swing2LimitDegrees=30, twistLimitDegrees=45}
 {bone1Name=dyn_lock_of_hair_3_3_l, bone2Name=dyn_lock_of_hair_3_2_l, swing1LimitDegrees=45, swing2LimitDegrees=45, twistLimitDegrees=30}
 {bone1Name=dyn_lock_of_hair_3_1_r, bone2Name=dyn_lock_of_hair_3_base_r, swing1LimitDegrees=5, swing2LimitDegrees=5, twistLimitDegrees=5}
 {bone1Name=dyn_lock_of_hair_3_2_r, bone2Name=dyn_lock_of_hair_3_1_r, swing1LimitDegrees=30, swing2LimitDegrees=20, twistLimitDegrees=20}
 {bone1Name=dyn_necklace_1, bone2Name=dyn_necklace_base, swing1LimitDegrees=45, swing2LimitDegrees=30, twistLimitDegrees=45}
 {bone1Name=dyn_lock_of_hair_1_base_r, bone2Name=dyn_lock_of_hair_1_1_r, swing1LimitDegrees=20, swing2LimitDegrees=20, twistLimitDegrees=20}
 {bone1Name=dyn_lock_of_hair_1_l, bone2Name=dyn_lock_of_hair_1_base_l, swing1LimitDegrees=20, swing2LimitDegrees=20, twistLimitDegrees=10}
 {
  bone1Name=dyn_necklace_base
  bone2Name=spine_05
  swing1LimitDegrees=45
  swing1Motion=Locked
  swing2LimitDegrees=45
  swing2Motion=Locked
  twistLimitDegrees=45
  twistMotion=Locked
 }
]
