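# Post-copy animation Blueprint for the VibrantShell female face accessory, simulating its hair tail and adjusting it around back bling.
# The pose copied from the attached parent runs through RigidBody physics in neck_02 space, with gravity updated each frame from the character's parameter asset.
# When back bling or a cape is detected, one alpha enables constraints to neck and spine bones plus additive bone offsets that reposition tail segments 2–4.
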AnimBlueprint /BRCosmetics/Characters/Player/Female/Medium/Bodies/F_MED_VibrantShell/Meshes/Parts/F_MED_VibrantShell_FaceAcc_AnimBP
target_skeleton=&/F_MED_VibrantShell_FaceAcc_Skeleton
graphs [
 AnimGraph {nodes [
   n1 = "Transform (Modify) Bone - Bone: dyn_hair_tail_2"(Alpha=1, Translation="-1,2,0", Rotation="0,0,0", Scale="1,1,1") {rotationMode=BMM_Additive, translationMode=BMM_Additive, boneToModify {boneName=dyn_hair_tail_2}, translation=<-1 2>}
   n2 = "Transform (Modify) Bone - Bone: dyn_hair_tail_3"(ComponentPose=@n1, Alpha=1, Translation="0,4,0", Rotation="0,-20,0", Scale="1,1,1") {rotationMode=BMM_Additive, translationMode=BMM_Additive, boneToModify {boneName=dyn_hair_tail_3}, rotation {yaw=-20}, translation {y=4}}
   "Transform (Modify) Bone - Bone: dyn_hair_tail_4"(ComponentPose=@n2, Alpha=1, Translation="0,3,0", Rotation="20,-30,30", Scale="1,1,1") {
    rotationMode=BMM_Additive
    translationMode=BMM_Additive
    boneToModify {boneName=dyn_hair_tail_4}
    rotation=<rot 20 -30 30>
    translation {y=3}
   }
   n3 = RigidBody!(
    ComponentPose=Local To Component(LocalPose=Copy Pose From Mesh() {bUseAttachedParent=true})
    Alpha
    OverrideWorldGravity:Vector=Get Override World Gravity() -> (Override World Gravity:Vector)
    ExternalForce
   ) -> (Pose) {
    bOverrideWorldGravity=true
    overridePhysicsAsset=&/F_MED_VibrantShell_FaceAcc_Physics
    simulationSpace=BaseBoneSpace
    baseBoneRef {boneName=neck_02}
    componentAppliedLinearAccClamp=<750 750 750>
    overrideWorldGravity {z=-980}
    simSpaceSettings {dampingAlpha=0, maxAngularAcceleration=15, maxAngularVelocity=3.5, maxLinearAcceleration=1200, maxLinearVelocity=100, worldAlpha=1}
   }
   n4 = Get Dyn_backpack_constr() -> (Dyn_backpack_constr:double)
   n5 = "Constraint - dyn_hair_tail_2 (PP)"!(ComponentPose=@n3, Alpha=@n4, ConstraintWeights_0=0.75, ConstraintWeights_1:float=0.25) -> (Pose) {
    boneToModify {boneName=dyn_hair_tail_2}
    constraintSetup [{targetBone {boneName=neck_01}}, {targetBone {boneName=neck_02}}]
    constraintWeights [1,1]
   }
   n6 = "Constraint - dyn_hair_tail_3 (PP)"!(ComponentPose=@n5, Alpha=@n4, ConstraintWeights_0=0.5, ConstraintWeights_1:float=0.5) -> (Pose) {
    boneToModify {boneName=dyn_hair_tail_3}
    constraintSetup [{targetBone {boneName=neck_01}}, {targetBone {boneName=spine_05}}]
    constraintWeights [1,1]
   }
   n7 = "Constraint - dyn_hair_tail_4 (PP)"!(ComponentPose=@n6, Alpha=@n4, ConstraintWeights_0=0.5, ConstraintWeights_1:float=0.5) -> (Pose) {
    boneToModify {boneName=dyn_hair_tail_4}
    constraintSetup [{targetBone {boneName=spine_05}}, {targetBone {boneName=spine_04}}]
    constraintWeights [1,1]
   }
   n8 = "Transform (Modify) Bone - Bone: dyn_hair_tail_2"(ComponentPose=@n7, Alpha=@n4, Translation="0.5,0,2.5", Rotation="0,0,-10", Scale="1,1,1") {
    rotationMode=BMM_Additive
    rotationSpace=BCS_BoneSpace
    translationMode=BMM_Additive
    translationSpace=BCS_BoneSpace
    boneToModify {boneName=dyn_hair_tail_2}
    rotation {roll=-10}
    translation {x=0.5, z=2.5}
   }
   n9 = "Transform (Modify) Bone - Bone: dyn_hair_tail_3"(ComponentPose=@n8, Alpha=@n4, Translation="-2,0,2", Rotation="0,0,-5", Scale="1,1,1") {
    rotationMode=BMM_Additive
    rotationSpace=BCS_BoneSpace
    scaleSpace=BCS_BoneSpace
    translationMode=BMM_Additive
    translationSpace=BCS_BoneSpace
    boneToModify {boneName=dyn_hair_tail_3}
    rotation {roll=-5}
    translation {x=-2, z=2}
   }
   n10 = "Transform (Modify) Bone - Bone: dyn_hair_tail_4"(ComponentPose=@n9, Alpha=@n4, Translation="-1,-4.5,5", Rotation="-50,-10,-15", Scale="1,1,0.6") {
    rotationMode=BMM_Additive
    rotationSpace=BCS_BoneSpace
    scaleMode=BMM_Additive
    scaleSpace=BCS_BoneSpace
    translationMode=BMM_Additive
    translationSpace=BCS_BoneSpace
    boneToModify {boneName=dyn_hair_tail_4}
    rotation=<rot -50 -10 -15>
    scale {z=0.6}
    translation=<-1 -4.5 5>
   }
   Output Pose(Result=Component To Local(ComponentPose=@n10)) {layerGroup=None}
  ]}
 EventGraph {nodes [
   Event BlueprintUpdateAnimation()
   > n1 = Sequence()
   > n2 = "Gravity Override State Function (Parameter Asset) (Nativized)"(
    CharacterPartAnimInstance=Self-Reference()
    GameTimeInSeconds=GetGameTimeInSeconds()
    Parameters=/BRCosmetics/Characters/Player/Female/Medium/Bodies/F_MED_VibrantShell/Meshes/F_MED_VibrantShell_Parameters
   )
   > Set Override World Gravity(Override World Gravity:Vector=@n2) -> (Output_Get:Vector)
   n3 = CheckForBackbling()
   n4 = Branch(execute=@n1.then_1, Condition=OR Boolean(A=@n3, B=@n3.bIsBackblingACape))
   > Set Dyn_backpack_constr(Dyn_backpack_constr:double=1) -> (Output_Get:double)
   Set Dyn_backpack_constr(execute=@n4.else, Dyn_backpack_constr=0)
   NOT Boolean(A=@n3.bIsBackblingACape)
   AND Boolean(A=@n3, B=@n3.bIsBackblingACape)
  ]}
]
variables [
 Dyn_backpack_constr {type=number}
 Override World Gravity {type=Vector, default=<0 0 -980>}
]
